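# Movie Render Graph template config that provides a basic starting setup for rendering, with a Globals branch and a DefaultLayer branch.
# Globals chains Sampling Method, Warm Up Settings with emulated motion blur, Global Game Overrides, and a tagged Global Output Settings node into the output.
# The game overrides disable LODs and HLODs and flush streaming, shader and asset compilers.
# DefaultLayer holds only a Render Layer node, with no renderer or file output node.
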MovieGraphConfig /MovieRenderPipeline/BasicConfigTemplateGraph
members {
 inputs [
  MovieGraphInput_0 {class=MovieGraphInput}
  MovieGraphInput_1 {class=MovieGraphInput}
 ]
 outputs [
  MovieGraphOutput_0 {class=MovieGraphOutput}
  MovieGraphOutput_1 {class=MovieGraphOutput}
 ]
 variables []
}
ids=id
nodes [
 DefaultInputNode = Input() {
  is=title MovieGraphInputNode
  pins are=other
  also {
   outputs [
    MovieGraphPin_1 {properties {type=None, allow_multiple_connections=false, is_branch=true, label=Globals}}
    MovieGraphPin_0 {properties {type=DOUBLE, allow_multiple_connections=false, is_branch=true, label=DefaultLayer}}
   ]
   position [-416,0]
  }
 }
 n1 = Global Game Overrides(MovieGraphPin_0=Warm Up Settings(MovieGraphPin_0=Sampling Method(MovieGraphPin_0=@DefaultInputNode.Globals) {
    is=title MovieGraphSamplingMethodNode
    pins are=other
    bOverride_TemporalSampleCount=true
    also {
     inputs [MovieGraphPin_0 {properties {type=None, allow_multiple_connections=false, is_branch=true}}]
     outputs [MovieGraphPin_1 {properties {type=None, allow_multiple_connections=false, is_branch=true}}]
     position [-16,0]
    }
   }) {
   is=title MovieGraphWarmUpSettingNode
   pins are=other
   bEmulateMotionBlur=true
   bOverride_NumWarmUpFrames=true
   bOverride_bEmulateMotionBlur=true
   numWarmUpFrames=64
   also {
    inputs [MovieGraphPin_0 {properties {type=None, allow_multiple_connections=false, is_branch=true}}]
    outputs [MovieGraphPin_1 {properties {type=None, allow_multiple_connections=false, is_branch=true}}]
    position [256,0]
   }
  }) {
  is=title MovieGraphGlobalGameOverridesNode
  pins are=other
  bDisableHLODs=true
  bDisableLODs=true
  bFlushAssetCompiler=true
  bFlushGrassStreaming=true
  bFlushLevelStreaming=true
  bFlushShaderCompiler=true
  bFlushStreamingManagers=true
  bOverride_GameModeOverride=true
  bOverride_ScalabilityQualityLevel=true
  bOverride_VirtualTextureFeedbackFactor=true
  bOverride_bDisableHLODs=true
  bOverride_bDisableLODs=true
  bOverride_bDisableTextureStreaming=true
  bOverride_bFlushAssetCompiler=true
  bOverride_bFlushGrassStreaming=true
  bOverride_bFlushLevelStreaming=true
  bOverride_bFlushShaderCompiler=true
  bOverride_bFlushStreamingManagers=true
  also {
   inputs [MovieGraphPin_0 {properties {type=None, allow_multiple_connections=false, is_branch=true}}]
   outputs [MovieGraphPin_1 {properties {type=None, allow_multiple_connections=false, is_branch=true}}]
   position [528,0]
  }
 }
 DefaultOutputNode = Output(
  Globals=Global Output Settings(MovieGraphPin_0=@n1) {
   is=title MovieGraphGlobalOutputSettingNode
   pins are=other
   scriptTags [Default_GlobalOutput]
   also {
    inputs [MovieGraphPin_0 {properties {type=None, allow_multiple_connections=false, is_branch=true}}]
    outputs [MovieGraphPin_1 {properties {type=None, allow_multiple_connections=false, is_branch=true}}]
    position [832,0]
   }
  }
  DefaultLayer=Render Layer(MovieGraphPin_0=@DefaultInputNode.DefaultLayer) {
   is=title MovieGraphRenderLayerNode
   pins are=other
   bOverride_LayerName=true
   also {
    inputs [MovieGraphPin_0 {properties {type=None, allow_multiple_connections=false, is_branch=true}}]
    outputs [MovieGraphPin_1 {properties {type=None, allow_multiple_connections=false, is_branch=true}}]
    position [896,176]
   }
  }
 ) {
  is=title MovieGraphOutputNode
  pins are=other
  also {
   inputs [
    MovieGraphPin_1 {properties {type=None, allow_multiple_connections=false, is_branch=true, label=Globals}}
    MovieGraphPin_0 {properties {type=DOUBLE, allow_multiple_connections=false, is_branch=true, label=DefaultLayer}}
   ]
   position [1168,0]
  }
 }
]
