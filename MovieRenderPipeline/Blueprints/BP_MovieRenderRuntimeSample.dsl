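; Sample Blueprint for runtime Movie Render Pipeline use through the MoviePipelineQueueEngineSubsystem.
; BeginPlay enables input and binds OnRenderFinished to print success; EndPlay cancels active executor jobs to avoid a render-thread deadlock.
; ConfigureSampleJob sets up 720p PNG output, two temporal samples and a deferred pass, but nothing here calls it.
; R, T and Y key events execute nothing, and a legacy advanced-executor finish handler is left unbound.
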(event EventBeginPlay
  ;[1 1 1] Listen to OnRenderFinished Callback (Used in simple examples)
  (EventDispatchers|BindEventtoOnRenderFinished (EngineSubsystems|GetMoviePipelineQueueEngineSubsystem) (delegate Custom|OnRenderFinished_Event))
  (Input|EnableInput (Game|GetPlayerController 0)))

(event Custom|OnRenderFinished_Event
  (params (Results :type Movie%20Pipeline%20Output%20Data%20Structure))
  (bind (_pipeline _job _bsuccess) (Utilities|Struct|BreakMoviePipelineOutputData Results))
  ;[1 1 1] Listen to OnRenderFinished Callback (Used in simple examples)
  (Development|PrintText (Utilities|Text|FormatText "NSLOCTEXT(\"[DE2363934955B3098E90ABB27806CFFB]\", \"D3ABD7CD453FB8C9929335A9F65BB4B4\", \"Render Finished. Success: {success}\")" _bsuccess)))

(event EventEndPlay
  (params (EndPlayReason :type EEndPlayReason%20Enum :default "Destroyed"))
  (bind _returnvalue (EngineSubsystems|GetMoviePipelineQueueEngineSubsystem))
  (bind _returnvalue_2 (MovieRenderPipeline|GetActiveExecutor _returnvalue))
  ;[1 1 1] Cancel any active jobs when the map ends (required to avoid render thread deadlock)
  (Utilities|IsValid _returnvalue_2
    (:Is%20Valid
      (MovieRenderPipeline|CancelAllJobs _returnvalue_2))))

(event R
  (params (Key :type Key%20Structure))
  (bind _returnvalue (EngineSubsystems|GetMoviePipelineQueueEngineSubsystem)))

(event T
  (params (Key :type Key%20Structure))
  (bind _returnvalue (EngineSubsystems|GetMoviePipelineQueueEngineSubsystem)))

(event Y
  (params (Key :type Key%20Structure))
  (bind _returnvalue (EngineSubsystems|GetMoviePipelineQueueEngineSubsystem))
  (bind _returnvalue_2 (MovieRenderPipeline|GetQueue _returnvalue)))

(event Custom|OnAdvancedExecutorFinished
  (params (PipelineExecutor :type Movie%20Pipeline%20Executor%20Base%20Object%20Reference) (bSuccess :type Boolean))
  ;[1 0 0] LEGACY EXAMPLE (2 sequences in Queue)
  (Development|PrintText (Utilities|Text|FormatText "NSLOCTEXT(\"[DE2363934955B3098E90ABB27806CFFB]\", \"FAACE7CA4D83884D7C39489B83BFF0A4\", \"Advanced Render Finished. Success: {success}\")" bSuccess)))

(fn ConfigureSampleJob
  (params (Target :type Movie%20Pipeline%20Executor%20Job%20Object%20Reference))
  (bind _returnvalue (MovieRenderPipeline|GetConfiguration Target))
  (bind _returnvalue_2 (MovieRenderPipeline|GetConfiguration Target))
  (Utilities|FlowControl|Sequence
    (:then_0
      (bind _returnvalue_3 (MovieRenderPipeline|FindorAddSettingbyClass _returnvalue_2 "/Script/MovieRenderPipelineCore.MoviePipelineOutputSetting" false false))
      (Variables|/Script/MovieRenderPipelineCore.MoviePipelineOutputSetting|SetOutputResolution (Utilities|Struct|MakeIntPoint 1280 720) _returnvalue_3)
      (Variables|/Script/MovieRenderPipelineCore.MoviePipelineOutputSetting|SetFileNameFormat "{sequence_name}/{frame_number}" _returnvalue_3))
    (:then_1
      (bind _returnvalue_4 (MovieRenderPipeline|FindorAddSettingbyClass _returnvalue_2 "/Script/MovieRenderPipelineCore.MoviePipelineAntiAliasingSetting" false false))
      (Variables|/Script/MovieRenderPipelineCore.MoviePipelineAntiAliasingSetting|SetTemporalSampleCount 2 _returnvalue_4))
    (:then_2
      ;[1 1 1] PNG Output + Deferred Renderer
      (MovieRenderPipeline|FindorAddSettingbyClass _returnvalue "/Script/MovieRenderPipelineRenderPasses.MoviePipelineImageSequenceOutput_PNG" false false))
    (:then_3
      (MovieRenderPipeline|FindorAddSettingbyClass _returnvalue "/Script/MovieRenderPipelineRenderPasses.MoviePipelineDeferredPassBase" false false))))

(fn ConstructionScript)
