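# Default Movie Render Pipeline burn-in widget that overlays job, build, timing, shot and camera metadata on rendered frames.
# OnOutputFrameStarted runs a Sequence of update functions, each pulling pipeline data into a text block.
# Top and bottom translucent bands show job/world/author, branch-changelist, root timecode, lens settings and shot timecode, with the user's job comment centered between them.
# The branch name is parsed from the engine version string after its last '+'.
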WidgetBlueprint /MovieRenderPipeline/Blueprints/DefaultBurnIn
graphs [
 EventGraph {nodes [
   n1 = Event OnOutputFrameStarted!() -> (OutputDelegate, then, ForPipeline:Movie Pipeline Object Reference) {is=title K2Node_Event}
   > n2 = Sequence!(execute) -> (then_0, then_1, then_2:Exec, then_3:Exec, then_4:Exec, then_5:Exec)
   > UpdateTopLeftText!(execute, self, ForPipeline:Movie Pipeline Object Reference=@n1.ForPipeline) -> (then)
   UpdateTopRightText!(execute=@n2.then_1, self, ForPipeline:Movie Pipeline Object Reference=@n1.ForPipeline) -> (then)
   UpdateBottomLeftText!(execute=@n2.then_2, self, ForPipeline:Movie Pipeline Object Reference=@n1.ForPipeline) -> (then)
   UpdateBottomCenter!(execute=@n2.then_3, self, ForPipeline:Movie Pipeline Object Reference=@n1.ForPipeline) -> (then)
   UpdateBottomRightText!(execute=@n2.then_4, self, ForPipeline:Movie Pipeline Object Reference=@n1.ForPipeline) -> (then)
   UpdateUserCommentText!(execute=@n2.then_5, self, For Pipeline:Movie Pipeline Object Reference=@n1.ForPipeline) -> (then)
  ]}
 UpdateBottomCenter {nodes [
   n1 = UpdateBottomCenter!() -> (then, ForPipeline:Movie Pipeline Object Reference) {is=title K2Node_FunctionEntry}
   > "SetText (Text)"(
    self=Get BottomCenter() -> (BottomCenter:Text Object Reference)
    InText=Format Text!(
     Format="LOCTEXT(\"{FocalLength}mm f/{FStop} {FocusDistance}\")"
     FocalLength:float=GetCurrentFocalLength!(InMoviePipeline:Movie Pipeline Object Reference=@n1) -> (ReturnValue:float)
     FStop:float=GetCurrentAperture!(InMoviePipeline:Movie Pipeline Object Reference=@n1) -> (ReturnValue:float)
     FocusDistance:float=GetCurrentFocusDistance!(InMoviePipeline:Movie Pipeline Object Reference=@n1) -> (ReturnValue:float)
    ) -> (Result)
   )
  ]}
 UpdateBottomLeftText {nodes [
   n1 = UpdateBottomLeftText!() -> (then, ForPipeline:Movie Pipeline Object Reference) {is=title K2Node_FunctionEntry}
   n2 = GetRootFrameNumber!(InMoviePipeline:Movie Pipeline Object Reference=@n1) -> (ReturnValue:Frame Number Structure)
   n3 = Timecode To String!(
    InTimecode:"Timecode Structure (by ref)"=GetRootTimecode!(InMoviePipeline:Movie Pipeline Object Reference=@n1) -> (ReturnValue:Timecode Structure)
    bForceSignDisplay:bool=false
   ) -> (ReturnValue:String)
   > "SetText (Text)"(
    self=Get BottomLeft() -> (BottomLeft:Text Object Reference)
    InText=Format Text!(
     Format="LOCTEXT(\"Timecode: {MasterTimecode} ({MasterFrameNumber}f)\")"
     MasterTimecode:Text="To Text (String)"(InString=@n3)
     MasterFrameNumber:int=Break Frame Number!(FrameNumber:"Frame Number Structure (by ref)"=@n2) -> (Value:int) {is=title K2Node_BreakStruct}
    ) -> (Result)
   )
  ]}
 UpdateBottomRightText {nodes [
   n1 = UpdateBottomRightText!() -> (then, ForPipeline:Movie Pipeline Object Reference) {is=title K2Node_FunctionEntry}
   n2 = GetCurrentShotFrameNumber!(InMoviePipeline:Movie Pipeline Object Reference=@n1) -> (ReturnValue:Frame Number Structure)
   n3 = Timecode To String(
    InTimecode=GetCurrentShotTimecode!(InMoviePipeline:Movie Pipeline Object Reference=@n1) -> (ReturnValue:Timecode Structure)
    bForceSignDisplay=false
   )
   > "SetText (Text)"(
    self=Get BottomRight() -> (BottomRight:Text Object Reference)
    InText=Format Text!(
     Format="LOCTEXT(\"Shot: {ShotName} {ShotTimecode} ({ShotFrameNumber}f)\")"
     ShotName:Text=GetCurrentSegmentName!(InMoviePipeline:Movie Pipeline Object Reference=@n1) -> (OutOuterName:Text, OutInnerName:Text)
     ShotTimecode:Text="To Text (String)"(InString=@n3)
     ShotFrameNumber:int=Break Frame Number(FrameNumber=@n2)
    ) -> (Result)
   )
  ]}
 UpdateTopLeftText {nodes [
   n1 = UpdateTopLeftText!() -> (then, ForPipeline:Movie Pipeline Object Reference) {is=title K2Node_FunctionEntry}
   > n2 = GetCurrentLevelName!(execute, bRemovePrefixString:bool=true) -> (then, ReturnValue:String)
   n3 = GetJobInitializationTime!(InMoviePipeline:Movie Pipeline Object Reference=@n1) -> (ReturnValue:Date Time Structure)
   > "SetText (Text)"(
    self=Get TopLeft() -> (TopLeft:Text Object Reference)
    InText=Format Text!(
     Format="LOCTEXT(\"Job: {SequenceName} {Time} {Date} World: {WorldName} Author: {AuthorName}\")"
     SequenceName:Text=GetJobName!(InMoviePipeline:Movie Pipeline Object Reference=@n1) -> (ReturnValue:Text)
     Time:Text="As Time (from UTC)"!(InDateTime:"Date Time Structure (by ref)"=@n3, InTimeZone:String, InTimeStyle:EDateTimeStyle Enum=Default) -> (ReturnValue:Text)
     Date:Text="As Date (from UTC)"!(InDateTime:"Date Time Structure (by ref)"=@n3, InTimeZone:String, InDateStyle:EDateTimeStyle Enum=Default) -> (ReturnValue:Text)
     WorldName:Text="To Text (String)"(InString=@n2)
     AuthorName:Text=GetJobAuthor!(InMoviePipeline:Movie Pipeline Object Reference=@n1) -> (ReturnValue:Text)
    ) -> (Result)
   )
  ]}
 UpdateTopRightText {nodes [
   n1 = UpdateTopRightText!() -> (then, ForPipeline:Movie Pipeline Object Reference) {is=title K2Node_FunctionEntry}
   > n2 = Sequence!(execute) -> (then_0, then_1, then_2)
   > Set ChangeList(ChangeList:String="To String (Text)"(InText=GetMoviePipelineEngineChangelistLabel!(InMoviePipeline:Movie Pipeline Object Reference=@n1) -> (ReturnValue:Text))) -> (Output_Get:String)
   n3 = Get BranchName() -> (BranchName:String)
   n4 = Get ChangeList() -> (ChangeList:String)
   "SetText (Text)"(
    execute=@n2.then_2
    self=Get TopRight() -> (TopRight:Text Object Reference)
    InText=Format Text!(
     Format="LOCTEXT(\"Branch: {BranchName}-{ChangeList}\")"
     BranchName:Text="To Text (String)"(InString=@n3)
     ChangeList:Text="To Text (String)"(InString=@n4)
    ) -> (Result)
   )
   n5 = GetEngineVersion!() -> (ReturnValue:String)
   n6 = FindSubstring(SearchIn=@n5, Substring=+, bSearchFromEnd=true)
   Branch(execute=@n2.then_1, Condition="integer >= integer"!(A:int=@n6, B:int=0) -> (ReturnValue:bool) {is=title K2Node_CallFunction})
   n7 = int + int(A=@n6, B=1) -> (ReturnValue=0) {is=title K2Node_CommutativeAssociativeBinaryOperator}
   > Set BranchName(BranchName:String=GetSubstring!(SourceString:String=@n5, StartIndex:int=@n7, Length:int=int - int(A=@n6, B=1) {is=title K2Node_CallFunction}) -> (ReturnValue:String)) -> (Output_Get:String)
   int - int(A=Len(S=@n5), B=@n7)
  ]}
 UpdateUserCommentText {nodes [
   n1 = UpdateUserCommentText!() -> (then, For Pipeline:Movie Pipeline Object Reference) {is=title K2Node_FunctionEntry}
   n2 = GetCurrentJob!(self:Movie Pipeline Object Reference=@n1) -> (ReturnValue:Movie Pipeline Executor Job Object Reference)
   > "SetText (Text)"(
    self=Get UserJobComment() -> (UserJobComment:Text Object Reference)
    InText="To Text (String)"(InString=Get Comment!(self:Movie Pipeline Executor Job Object Reference=@n2) -> (Comment:String))
   )
  ]}
]
variables []
widgets {
 info {
  inheritedWidgetCount=0
  parentClass=/Script/MovieRenderPipelineSettings.MoviePipelineBurnInWidget
  rootWidgetClass=&Overlay
  widgetCount=16
 }
 widgets [
  Overlay_45 {widgetClassPath=&Overlay, widget=Overlay_45}
  RenderedImagePlaceholder {
   widgetClassPath=&Image
   parent=Overlay_45
   widget=RenderedImagePlaceholder
   properties {
    visibility=Hidden
    brush {
     bIsDynamicallyLoaded=false
     drawAs=Image
     imageType=NoImage
     mirroring=NoMirror
     resourceName=None
     resourceObject=null
     tiling=NoTile
     imageSize=<1920 1080>
     margin=<ltrb 0 0 0 0>
     outlineSettings {
      bUseBrushTransparency=false
      roundingType=HalfHeightRadius
      width=0
      color {colorUseRule=UseColor_Specified, specifiedColor=<rgba 0 0 0 0>}
      cornerRadii=<0 0 0 0>
     }
     tintColor {colorUseRule=UseColor_Specified, specifiedColor=<rgba 1 1 1 1>}
     uVRegion {bIsValid=false, max=<0 0>, min=<0 0>}
    }
   }
   slot {class=OverlaySlot, properties {horizontalAlignment=HAlign_Fill, verticalAlignment=VAlign_Fill}}
  }
  VerticalBox_41 {
   widgetClassPath=&VerticalBox
   parent=Overlay_45
   widget=VerticalBox_41
   slot {class=OverlaySlot, properties {horizontalAlignment=HAlign_Fill, verticalAlignment=VAlign_Fill}}
  }
  Border_23 {
   widgetClassPath=&Border
   parent=VerticalBox_41
   widget=Border_23
   properties {
    background {
     bIsDynamicallyLoaded=false
     drawAs=Image
     imageType=NoImage
     mirroring=NoMirror
     resourceName=None
     resourceObject=null
     tiling=NoTile
     imageSize=<32 32>
     margin=<ltrb 0 0 0 0>
     outlineSettings {
      bUseBrushTransparency=false
      roundingType=HalfHeightRadius
      width=0
      color {colorUseRule=UseColor_Specified, specifiedColor=<rgba 0 0 0 0>}
      cornerRadii=<0 0 0 0>
     }
     tintColor {colorUseRule=UseColor_Specified, specifiedColor=<rgba 0 0 0 0.5>}
     uVRegion {bIsValid=false, max=<0 0>, min=<0 0>}
    }
    padding=<ltrb 4 4 4 4>
   }
   slot {class=VerticalBoxSlot, properties {verticalAlignment=VAlign_Top}}
  }
  HorizontalBox_22 {
   widgetClassPath=&HorizontalBox
   parent=Border_23
   widget=HorizontalBox_22
   slot {class=BorderSlot, properties {padding=<ltrb 4 4 4 4>}}
  }
  TopLeft {
   widgetClassPath=&TextBlock
   bIsVariable=true
   parent=HorizontalBox_22
   widget=TopLeft
   properties {
    text=Job: PipelineTestSeq 03:44pm 01/19/2020 World: MoviePipelineTestMap Author: MannyMannequin
    font {
     bForceMonospaced=false
     bMaterialIsStencil=false
     fontMaterial=null
     fontObject=/Engine/EngineFonts/DroidSansMono
     letterSpacing=0
     monospacedWidth=1
     size=18
     skewAmount=0
     typefaceFontName=Font
     outlineSettings {
      bApplyOutlineToDropShadows=false
      bMiteredCorners=false
      bSeparateFillAlpha=false
      outlineMaterial=null
      outlineSize=0
      outlineColor=<rgba 0 0 0 1>
     }
    }
   }
   slot {class=HorizontalBoxSlot, properties {horizontalAlignment=HAlign_Left, size {sizeRule=Fill, value=0.5}}}
  }
  TopCenter {
   widgetClassPath=&TextBlock
   parent=HorizontalBox_22
   widget=TopCenter
   properties {font {
     bForceMonospaced=false
     bMaterialIsStencil=false
     fontMaterial=null
     fontObject=&DroidSansMono
     letterSpacing=0
     monospacedWidth=1
     size=18
     skewAmount=0
     typefaceFontName=Font
     outlineSettings {
      bApplyOutlineToDropShadows=false
      bMiteredCorners=false
      bSeparateFillAlpha=false
      outlineMaterial=null
      outlineSize=0
      outlineColor=<rgba 0 0 0 1>
     }
    }}
   slot {class=HorizontalBoxSlot, properties {horizontalAlignment=HAlign_Center}}
  }
  TopRight {
   widgetClassPath=&TextBlock
   bIsVariable=true
   parent=HorizontalBox_22
   widget=TopRight
   properties {
    justification=Right
    text=Branch: Dev-Editor-11047590
    font {
     bForceMonospaced=false
     bMaterialIsStencil=false
     fontMaterial=null
     fontObject=&DroidSansMono
     letterSpacing=0
     monospacedWidth=1
     size=18
     skewAmount=0
     typefaceFontName=Font
     outlineSettings {
      bApplyOutlineToDropShadows=false
      bMiteredCorners=false
      bSeparateFillAlpha=false
      outlineMaterial=null
      outlineSize=0
      outlineColor=<rgba 0 0 0 1>
     }
    }
   }
   slot {class=HorizontalBoxSlot, properties {horizontalAlignment=HAlign_Right, size {sizeRule=Fill, value=0.5}}}
  }
  Spacer_60 {
   widgetClassPath=&Spacer
   parent=VerticalBox_41
   widget=Spacer_60
   slot {class=VerticalBoxSlot, properties {size {sizeRule=Fill, value=1}}}
  }
  Border_70 {
   widgetClassPath=&Border
   parent=VerticalBox_41
   widget=Border_70
   properties {
    horizontalAlignment=HAlign_Center
    background {
     bIsDynamicallyLoaded=false
     drawAs=Image
     imageType=NoImage
     mirroring=NoMirror
     resourceName=None
     resourceObject=null
     tiling=NoTile
     imageSize=<32 32>
     margin=<ltrb 0 0 0 0>
     outlineSettings {
      bUseBrushTransparency=false
      roundingType=HalfHeightRadius
      width=0
      color {colorUseRule=UseColor_Specified, specifiedColor=<rgba 0 0 0 0>}
      cornerRadii=<0 0 0 0>
     }
     tintColor {colorUseRule=UseColor_Specified, specifiedColor=<rgba 1 1 1 0>}
     uVRegion {bIsValid=false, max=<0 0>, min=<0 0>}
    }
    padding=<ltrb 0 0 0 0>
   }
   slot {class=VerticalBoxSlot, properties {verticalAlignment=VAlign_Bottom, size {sizeRule=Fill, value=1}}}
  }
  UserJobComment {
   widgetClassPath=&TextBlock
   bIsVariable=true
   parent=Border_70
   widget=UserJobComment
   properties {
    autoWrapText=true
    justification=Center
    text="This is a really long comment entered by a user on a MRQ job. This field could be used for making notes about what they changed in this shot so that the director or lead can know what parts of the shot to pay attention to, and so you can remember what was changed after it goes into an edit. This is a long field that has multiple lines of text to show multi-line handling. If the comment field is left empty on the job then this will not be visible.\r\n\r\nYou can even use Shift+Enter to create multiple lines.\r\n\r\nLike this!\r\n\r\nOr this!\r\n\r\nReally tall text!"
    font {
     bForceMonospaced=false
     bMaterialIsStencil=false
     fontMaterial=null
     fontObject=&DroidSansMono
     letterSpacing=0
     monospacedWidth=1
     size=18
     skewAmount=0
     typefaceFontName=Default
     outlineSettings {
      bApplyOutlineToDropShadows=false
      bMiteredCorners=false
      bSeparateFillAlpha=false
      outlineMaterial=null
      outlineSize=1
      outlineColor=<rgba 0 0 0 1>
     }
    }
    shadowOffset=<0 0>
   }
   slot {class=BorderSlot, properties {horizontalAlignment=HAlign_Center, padding=<ltrb 0 0 0 0>}}
  }
  Border {
   widgetClassPath=&Border
   parent=VerticalBox_41
   widget=Border
   properties {
    background {
     bIsDynamicallyLoaded=false
     drawAs=Image
     imageType=NoImage
     mirroring=NoMirror
     resourceName=None
     resourceObject=null
     tiling=NoTile
     imageSize=<32 32>
     margin=<ltrb 0 0 0 0>
     outlineSettings {
      bUseBrushTransparency=false
      roundingType=HalfHeightRadius
      width=0
      color {colorUseRule=UseColor_Specified, specifiedColor=<rgba 0 0 0 0>}
      cornerRadii=<0 0 0 0>
     }
     tintColor {colorUseRule=UseColor_Specified, specifiedColor=<rgba 0 0 0 0.5>}
     uVRegion {bIsValid=false, max=<0 0>, min=<0 0>}
    }
    padding=<ltrb 4 4 4 4>
   }
   slot {class=VerticalBoxSlot, properties {verticalAlignment=VAlign_Top}}
  }
  HorizontalBox {widgetClassPath=&HorizontalBox, parent=Border, widget=HorizontalBox, slot {class=BorderSlot, properties {padding=<ltrb 4 4 4 4>}}}
  BottomLeft {
   widgetClassPath=&TextBlock
   bIsVariable=true
   parent=HorizontalBox
   widget=BottomLeft
   properties {
    text="Timecode: [00:00:14:8] (344f)"
    font {
     bForceMonospaced=false
     bMaterialIsStencil=false
     fontMaterial=null
     fontObject=&DroidSansMono
     letterSpacing=0
     monospacedWidth=1
     size=18
     skewAmount=0
     typefaceFontName=Font
     outlineSettings {
      bApplyOutlineToDropShadows=false
      bMiteredCorners=false
      bSeparateFillAlpha=false
      outlineMaterial=null
      outlineSize=0
      outlineColor=<rgba 0 0 0 1>
     }
    }
   }
   slot {class=HorizontalBoxSlot, properties {horizontalAlignment=HAlign_Left, size {sizeRule=Fill, value=0.5}}}
  }
  BottomCenter {
   widgetClassPath=&TextBlock
   bIsVariable=true
   parent=HorizontalBox
   widget=BottomCenter
   properties {
    text=60mm f/2.5
    font {
     bForceMonospaced=false
     bMaterialIsStencil=false
     fontMaterial=null
     fontObject=&DroidSansMono
     letterSpacing=0
     monospacedWidth=1
     size=18
     skewAmount=0
     typefaceFontName=Font
     outlineSettings {
      bApplyOutlineToDropShadows=false
      bMiteredCorners=false
      bSeparateFillAlpha=false
      outlineMaterial=null
      outlineSize=0
      outlineColor=<rgba 0 0 0 1>
     }
    }
   }
   slot {class=HorizontalBoxSlot, properties {horizontalAlignment=HAlign_Center}}
  }
  BottomRight {
   widgetClassPath=&TextBlock
   bIsVariable=true
   parent=HorizontalBox
   widget=BottomRight
   properties {
    justification=Right
    text="Shot: RenderFarmSeqB Cam: CineCameraActor1 [00:00:3:13] (109f)"
    font {
     bForceMonospaced=false
     bMaterialIsStencil=false
     fontMaterial=null
     fontObject=&DroidSansMono
     letterSpacing=0
     monospacedWidth=1
     size=18
     skewAmount=0
     typefaceFontName=Font
     outlineSettings {
      bApplyOutlineToDropShadows=false
      bMiteredCorners=false
      bSeparateFillAlpha=false
      outlineMaterial=null
      outlineSize=0
      outlineColor=<rgba 0 0 0 1>
     }
    }
   }
   slot {class=HorizontalBoxSlot, properties {horizontalAlignment=HAlign_Right, size {sizeRule=Fill, value=0.5}}}
  }
 ]
}
