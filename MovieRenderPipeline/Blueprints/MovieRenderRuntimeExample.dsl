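; Runtime example Blueprint for starting Movie Render Pipeline renders through the engine subsystem.
; RenderQueue copies a supplied queue into the Movie Pipeline Queue Engine Subsystem's own queue, then renders it with the given executor class.
; RenderSequence declares inputs for a soft level sequence, map, executor type and use-current-level flag, but has no body here.
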(fn RenderQueue
  (params (Queue :type Movie%20Pipeline%20Queue%20Object%20Reference) (Executor%20Type :type Movie%20Pipeline%20Executor%20Base%20Class%20Reference))
  (bind _returnvalue (EngineSubsystems|GetMoviePipelineQueueEngineSubsystem))
  (MovieRenderPipeline|Queue|CopyFrom (MovieRenderPipeline|GetQueue _returnvalue) Queue)
  (MovieRenderPipeline|Rendering|RenderQueuewithExecutor _returnvalue Executor%20Type))

(fn RenderSequence
  (params (Sequence :type Level%20Sequence%20Soft%20Object%20Reference) (Map :type Soft%20Object%20Reference) (Executor%20Type :type Movie%20Pipeline%20Executor%20Base%20Class%20Reference) (UseCurrentLevel :type Boolean)))
