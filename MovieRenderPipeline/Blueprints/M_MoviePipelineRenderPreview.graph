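# UI-domain material for displaying the Movie Render Pipeline render preview, using alpha composite blending.
# A RenderTexture parameter supplies the image, sampled as linear color with an engine grey texture as the default.
# A OneMinus node inverts the sampled alpha channel.
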Material /MovieRenderPipeline/Blueprints/M_MoviePipelineRenderPreview
graph {
 properties {blendMode=BLEND_AlphaComposite, materialDomain=MD_UI, pixelDepthOffsetMode=PDOM_Legacy, usedShadingModels=Default Lit}
 nodes [OneMinus(None=TextureSampleParameter2D() {parameterName=RenderTexture, samplerType=SAMPLERTYPE_LinearColor, texture=/Engine/ArtTools/RenderToTexture/Textures/127grey}.A)]
}
