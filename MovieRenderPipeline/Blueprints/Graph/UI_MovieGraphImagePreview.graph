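# Movie Render Graph widget that previews one render layer's image, scaled down to fit over a tiled transparency checkerboard.
# OnInitialized creates a dynamic material for the preview brush.
# SetPreviewImage feeds the texture into that material, then, for render targets only, sizes both images to match, labels the layer (plus camera name if several) and rebuilds banner message entries.
# GetTextureRepresented returns the stored texture.
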WidgetBlueprint /MovieRenderPipeline/Blueprints/Graph/UI_MovieGraphImagePreview
graphs [
 EventGraph {nodes [
   n1 = SetPreviewImage!() -> (OutputDelegate, then, Data:Movie Graph Image Preview Data Structure) {is=title K2Node_CustomEvent}
   n2 = Break Movie Graph Image Preview Data!(MovieGraphImagePreviewData:"Movie Graph Image Preview Data Structure (by ref)"=@n1.Data) -> (
    Texture:Texture Object Reference
    Identifier:Movie Graph Render Data Identifier Structure
    bMultipleCameraNames:bool=~
    BannerMessages:Array of Movie Graph Preview Banner Message Structures
   ) {is=title K2Node_BreakStruct}
   > Set TextureRepresented(TextureRepresented:Texture Object Reference=@n2) -> (Output_Get:Texture Object Reference)
   > SetTextureParameterValue(
    self=Get DynamicMaterialInstance() -> (DynamicMaterialInstance:Material Instance Dynamic Object Reference)
    ParameterName=RenderTexture
    Value=@n2
   )
   > n3 = Cast To TextureRenderTarget2D!(execute, Object=@n2) -> (then, CastFailed, AsTexture Render Target 2D:Texture Render Target 2D Object Reference) {is=title K2Node_DynamicCast}
   > n4 = Sequence!(execute) -> (then_0, then_1, then_2:Exec, then_3:Exec)
   n5 = MakeVector2D(
    X="To Float (Integer)"(InInt=Get SizeX!(self:Texture Render Target 2D Object Reference=@n3) -> (SizeX:int=0))
    Y="To Float (Integer)"(InInt=Get SizeY!(self:Texture Render Target 2D Object Reference=@n3) -> (SizeY:int=0))
   )
   n6 = Get PreviewImage() -> (PreviewImage:Image Object Reference)
   > n7 = Set members in Slate Brush!(
    execute:Exec
    StructRef:"Slate Brush Structure (by ref)"=Get Brush!(self:Image Object Reference=@n6) -> (Brush:Slate Brush Structure)
    ImageSize:Vector 2D Structure=@n5
   ) -> (then:Exec, StructOut:"Slate Brush Structure (by ref)") {is=title K2Node_SetFieldsInStruct}
   > SetBrush(self=@n6, InBrush=@n7)
   > SetVisibility(self=@n6, InVisibility=Visible)
   n8 = Get TransparentCheckerboard() -> (TransparentCheckerboard:Image Object Reference)
   n9 = Set members in Slate Brush(execute=@n4.then_1, StructRef=Get Brush(self=@n8), ImageSize=@n5)
   > SetBrush(self=@n8, InBrush=@n9)
   > SetVisibility(self=@n8, InVisibility=Visible)
   n10 = Break Movie Graph Render Data Identifier!(MovieGraphRenderDataIdentifier:"Movie Graph Render Data Identifier Structure (by ref)"=@n2.Identifier) -> (RootBranchName:Name, LayerName:String, RendererName:String, SubResourceName:String, CameraName:String) {is=title K2Node_BreakStruct}
   n11 = "To Text (String)"(InString=@n10.LayerName)
   "SetText (Text)"(
    execute=@n4.then_2
    self=Get LayerNameTextBlock() -> (LayerNameTextBlock:Text Object Reference)
    InText=SelectText(
     A=Format Text!(Format="LOCTEXT(\"{LayerName} - {CameraName}\")", LayerName:Text=@n11, CameraName:String=@n10.CameraName) -> (Result)
     B=@n11
     bPickA=@n2.bMultipleCameraNames
    )
   )
   Reset!(
    execute=@n4.then_3
    self:Dynamic Entry Box Object Reference=Get BannerMessages() -> (BannerMessages:Dynamic Entry Box Object Reference)
    bDeleteWidgets:bool=false
   ) -> (then)
   > n12 = For Each Loop(Array:Array of Movie Graph Preview Banner Message Structures=@n2.BannerMessages) -> (Array Element:Movie Graph Preview Banner Message Structure)
   > n13 = Create Entry!(execute, self:Dynamic Entry Box Object Reference=Get BannerMessages()) -> (then, ReturnValue:User Widget Object Reference)
   > n14 = Cast To UI_MovieGraph_PreviewWindowBannerMessage!(execute, Object=@n13) -> (then, CastFailed, AsUI Movie Graph Preview Window Banner Message:UI Movie Graph Preview Window Banner Message Object Reference) {is=title K2Node_DynamicCast}
   > SetMessage!(execute, self:UI Movie Graph Preview Window Banner Message Object Reference=@n14, Message:Movie Graph Preview Banner Message Structure=@n12) -> (then)
   Event OnInitialized()
   > n15 = CreateDynamicMaterialInstance!(execute, Parent=/MovieRenderPipeline/Blueprints/M_MoviePipelineRenderPreview, OptionalName=None, CreationFlags=None) -> (then, ReturnValue)
   > n16 = Set DynamicMaterialInstance(DynamicMaterialInstance:Material Instance Dynamic Object Reference=@n15) -> (Output_Get:Material Instance Dynamic Object Reference)
   > SetBrushFromMaterial(self=Get PreviewImage(), Material=@n16)
  ]}
 GetTextureRepresented {nodes [
   GetTextureRepresented() {is=title K2Node_FunctionEntry}
   > Return Node!(execute, Texture:Texture Object Reference=Get TextureRepresented() -> (TextureRepresented:Texture Object Reference))
  ]}
]
variables [
 DynamicMaterialInstance {type=MaterialInstanceDynamic}
 TextureRepresented {type=Texture, default=null}
]
widgets {
 info {inheritedWidgetCount=0, parentClass=&UserWidget, rootWidgetClass=&HorizontalBox, widgetCount=9}
 widgets [
  HorizontalBox_126 {widgetClassPath=&HorizontalBox, widget=HorizontalBox_126}
  ScaleBox_37 {
   widgetClassPath=&ScaleBox
   parent=HorizontalBox_126
   widget=ScaleBox_37
   properties {stretchDirection=DownOnly}
   slot {class=HorizontalBoxSlot, properties {size {sizeRule=Fill, value=1}}}
  }
  Overlay_0 {widgetClassPath=&Overlay, parent=ScaleBox_37, widget=Overlay_0, slot {class=ScaleBoxSlot}}
  TransparentCheckerboard {
   widgetClassPath=&Image
   bIsVariable=true
   parent=Overlay_0
   widget=TransparentCheckerboard
   properties {
    visibility=Hidden
    brush {
     bIsDynamicallyLoaded=false
     drawAs=Image
     imageType=NoImage
     mirroring=NoMirror
     resourceName=None
     resourceObject=&/T_MovieGraph_TransparencyBG
     tiling=Both
     imageSize=<720 480>
     margin=<ltrb 0 0 0 0>
     outlineSettings {
      bUseBrushTransparency=false
      roundingType=HalfHeightRadius
      width=0
      color {colorUseRule=UseColor_Specified, specifiedColor=<rgba 0 0 0 0>}
      cornerRadii=<0 0 0 0>
     }
     tintColor {colorUseRule=UseColor_Specified, specifiedColor=<rgba 1 1 1 1>}
     uVRegion {bIsValid=false, max=<0 0>, min=<0 0>}
    }
   }
   slot {class=OverlaySlot}
  }
  PreviewImage {
   widgetClassPath=&Image
   bIsVariable=true
   parent=Overlay_0
   widget=PreviewImage
   properties {
    visibility=Hidden
    brush {
     bIsDynamicallyLoaded=false
     drawAs=Image
     imageType=NoImage
     mirroring=NoMirror
     resourceName=None
     resourceObject=null
     tiling=NoTile
     imageSize=<720 480>
     margin=<ltrb 0 0 0 0>
     outlineSettings {
      bUseBrushTransparency=false
      roundingType=HalfHeightRadius
      width=0
      color {colorUseRule=UseColor_Specified, specifiedColor=<rgba 0 0 0 0>}
      cornerRadii=<0 0 0 0>
     }
     tintColor {colorUseRule=UseColor_Specified, specifiedColor=<rgba 1 1 1 1>}
     uVRegion {bIsValid=false, max=<0 0>, min=<0 0>}
    }
   }
   slot {class=OverlaySlot}
  }
  HorizontalBox_219 {
   widgetClassPath=&HorizontalBox
   parent=Overlay_0
   widget=HorizontalBox_219
   slot {class=OverlaySlot, properties {horizontalAlignment=HAlign_Fill, verticalAlignment=VAlign_Bottom}}
  }
  Border_25 {
   widgetClassPath=&Border
   parent=HorizontalBox_219
   widget=Border_25
   properties {
    horizontalAlignment=HAlign_Left
    verticalAlignment=VAlign_Center
    background {
     bIsDynamicallyLoaded=false
     drawAs=Image
     imageType=NoImage
     mirroring=NoMirror
     resourceName=None
     resourceObject=null
     tiling=NoTile
     imageSize=<32 32>
     margin=<ltrb 0 0 0 0>
     outlineSettings {
      bUseBrushTransparency=false
      roundingType=HalfHeightRadius
      width=0
      color {colorUseRule=UseColor_Specified, specifiedColor=<rgba 0 0 0 0>}
      cornerRadii=<0 0 0 0>
     }
     tintColor {colorUseRule=UseColor_Specified, specifiedColor=<rgba 0 0 0 0.5>}
     uVRegion {bIsValid=false, max=<0 0>, min=<0 0>}
    }
    contentColorAndOpacity=<rgba 0.8698 0.8698 0.8698 1>
    padding=<ltrb 4 0 4 0>
   }
   slot {class=HorizontalBoxSlot, properties {horizontalAlignment=HAlign_Left, verticalAlignment=VAlign_Bottom}}
  }
  LayerNameTextBlock {
   widgetClassPath=&TextBlock
   bIsVariable=true
   parent=Border_25
   widget=LayerNameTextBlock
   properties {
    text=LAYER NAME
    font {
     bForceMonospaced=false
     bMaterialIsStencil=false
     fontMaterial=null
     fontObject=&Roboto
     letterSpacing=0
     monospacedWidth=1
     size=38
     skewAmount=0
     typefaceFontName=Bold
     outlineSettings {
      bApplyOutlineToDropShadows=false
      bMiteredCorners=false
      bSeparateFillAlpha=false
      outlineMaterial=null
      outlineSize=0
      outlineColor=<rgba 0 0 0 1>
     }
    }
   }
   slot {class=BorderSlot, properties {horizontalAlignment=HAlign_Left, verticalAlignment=VAlign_Center, padding=<ltrb 4 0 4 0>}}
  }
  BannerMessages {
   widgetClassPath=&DynamicEntryBox
   bIsVariable=true
   parent=HorizontalBox_219
   widget=BannerMessages
   properties {
    entryBoxType=Wrap
    entryHorizontalAlignment=HAlign_Left
    entryWidgetClass=&/UI_MovieGraph_PreviewWindowBannerMessage.UI_MovieGraph_PreviewWindowBannerMessage_C
    entrySpacing=<4 0>
   }
   slot {class=HorizontalBoxSlot, properties {padding=<ltrb 4 0 4 0>, size {sizeRule=Fill, value=1}}}
  }
 ]
}
