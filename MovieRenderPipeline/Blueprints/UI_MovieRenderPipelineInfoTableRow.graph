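# Widget Blueprint for a Movie Render Pipeline info table row, showing a left label and a right value.
# A horizontal box holds the two text blocks, separated by a fill spacer.
# Pre Construct and Construct both copy the LeftColumnText and RightColumnText variables into the blocks.
# SetValueText and SetValueColor events update only the right block's text and color, and Tick is empty.
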WidgetBlueprint /MovieRenderPipeline/Blueprints/UI_MovieRenderPipelineInfoTableRow
graphs [EventGraph {nodes [
   n1 = SetValueColor!() -> (OutputDelegate, then, Color:Slate Color Structure) {is=title K2Node_CustomEvent}
   n2 = Get RightTextBlock() -> (RightTextBlock:Text Object Reference)
   > SetColorAndOpacity(self=@n2, InColorAndOpacity=@n1.Color)
   n3 = SetValueText!() -> (OutputDelegate, then, Text:Text) {is=title K2Node_CustomEvent}
   > "SetText (Text)"(self=@n2, InText=@n3.Text)
   n4 = Event PreConstruct()
   "SetText (Text)"(
    execute=[@n4.then, @n5.then]
    self=Get LeftTextBlock() -> (LeftTextBlock:Text Object Reference)
    InText=Get LeftColumnText() -> (LeftColumnText:Text)
   )
   > "SetText (Text)"(self=Get RightTextBlock(), InText=Get RightColumnText() -> (RightColumnText:Text))
   n5 = Event Construct()
   Event Tick()
  ]}]
variables [
 LeftColumnText {type=string, default=Default Left Column Text}
 RightColumnText {type=string, default=Default Right Column Text}
]
widgets {
 info {inheritedWidgetCount=0, parentClass=&UserWidget, rootWidgetClass=&HorizontalBox, widgetCount=4}
 widgets [
  HorizontalBox_25 {widgetClassPath=&HorizontalBox, widget=HorizontalBox_25}
  LeftTextBlock {
   widgetClassPath=&TextBlock
   bIsVariable=true
   parent=HorizontalBox_25
   widget=LeftTextBlock
   properties {text=Default Left Column Text}
   slot {class=HorizontalBoxSlot, properties {horizontalAlignment=HAlign_Left}}
  }
  Spacer_37 {
   widgetClassPath=&Spacer
   parent=HorizontalBox_25
   widget=Spacer_37
   properties {size=<12 1>}
   slot {class=HorizontalBoxSlot, properties {size {sizeRule=Fill, value=1}}}
  }
  RightTextBlock {
   widgetClassPath=&TextBlock
   bIsVariable=true
   parent=HorizontalBox_25
   widget=RightTextBlock
   properties {text=Default Right Column Text}
   slot {class=HorizontalBoxSlot, properties {horizontalAlignment=HAlign_Left}}
  }
 ]
}
