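# Screen overlay widget for Movie Render Graph jobs, showing live render previews and pipeline progress.
# OnInitializedForPipeline stores the pipeline; each Tick switches off the Initializing page and refreshes job, frame, timing, cut and state rows, color-coding states and toggling warm-up and sub-sample rows.
# Renderer preview data fills a grid of image widgets mapped by render identifier, rebuilt only when identifiers change; otherwise only widgets with differing textures update.
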WidgetBlueprint /MovieRenderPipeline/Blueprints/UI_MovieGraphPipelineScreenOverlay
graphs [
 CreateEnoughGridWidgets {nodes [
   n1 = CreateEnoughGridWidgets!() -> (then, CurData:Array of Movie Graph Image Preview Data Structures) {is=title K2Node_FunctionEntry}
   > n2 = Set NumWidgetsToCreate(NumWidgetsToCreate:int=Length(TargetArray:Array of Movie Graph Image Preview Data Structures=@n1)) -> (Output_Get:int=0)
   > n3 = GetNumRowAndColumnForGrid!(execute, self, NumPreviews:int=@n2) -> (then, Rows:int=0, Columns:int=0)
   > n4 = For Each Loop(Array:Array of Movie Graph Image Preview Data Structures=@n1) -> (Array Element:Movie Graph Image Preview Data Structure)
   > n5 = Create Widget!(
    execute:Exec
    Class:User Widget Class Reference=/MovieRenderPipeline/Blueprints/Graph/UI_MovieGraphImagePreview.UI_MovieGraphImagePreview_C
    OwningPlayer:Player Controller Object Reference=GetOwningPlayer()
   ) -> (then:Exec, ReturnValue:UI Movie Graph Image Preview Object Reference)
   > Add!(
    execute
    TargetMap:Map of Movie Graph Render Data Identifier Structures to UI Movie Graph Image Preview Object References=Get CreatedGridWidgets() -> (CreatedGridWidgets:Map of Movie Graph Render Data Identifier Structures to UI Movie Graph Image Preview Object References)
    Key:"Movie Graph Render Data Identifier Structure (by ref)"=Break Movie Graph Image Preview Data!(MovieGraphImagePreviewData:"Movie Graph Image Preview Data Structure (by ref)"=@n4) -> (
     Texture:Texture Object Reference
     Identifier:Movie Graph Render Data Identifier Structure
     bMultipleCameraNames:bool=~
     BannerMessages:Array of Movie Graph Preview Banner Message Structures
    ) {is=title K2Node_BreakStruct}.Identifier
    Value:"UI Movie Graph Image Preview Object Reference (by ref)"=@n5
   ) -> (then) {is=title K2Node_CallFunction}
   > SetPreviewImage!(execute, self:UI Movie Graph Image Preview Object Reference=@n5, Data:Movie Graph Image Preview Data Structure=@n4) -> (then)
   > n6 = AddChildToUniformGrid!(
    execute
    self:Uniform Grid Panel Object Reference=Get PreviewImageGrid() -> (PreviewImageGrid:Uniform Grid Panel Object Reference)
    Content:Widget Object Reference=@n5
    InRow:int=int / int(A=@n4.Array Index, B=@n3.Columns)
    InColumn:int="% (Integer)"(A=@n4.Array Index, B=@n3.Columns) -> (ReturnValue=0)
   ) -> (then, ReturnValue:Uniform Grid Slot Object Reference)
   > SetHorizontalAlignment(self:Uniform Grid Slot Object Reference=@n6, InHorizontalAlignment=HAlign_Fill)
   > SetVerticalAlignment(self:Uniform Grid Slot Object Reference=@n6, InVerticalAlignment=VAlign_Fill)
  ]}
 EventGraph {nodes [
   Event Tick()
   > SetActiveWidgetIndex(self=Get WidgetSwitcher_0() -> (WidgetSwitcher_0:Widget Switcher Object Reference), Index=0)
   > n1 = Sequence!(execute) -> (then_0, then_1, then_2:Exec)
   > SetSequenceName()
   > SetTotalFrames()
   > SetElapsedTime()
   > SetEstimatedTime()
   > SetPipelineState_Text()
   > SetPipelineState_Color()
   > SetCurrentShotName()
   > SetCurrentLocalFrame()
   > SetSubSample()
   > SetWarmUpFrames()
   > SetShotState_Text()
   > SetShotState_Color()
   > UpdateRowWidgetVisibilities()
   n2 = GetPreviewData!(self:Movie Graph Renderer Base Object Reference=GetRendererInstance!(self:Movie Graph Pipeline Object Reference=Get Pipeline() -> (Pipeline:Movie Graph Pipeline Object Reference)) -> (ReturnValue:Movie Graph Renderer Base Object Reference)) -> (ReturnValue:Array of Movie Graph Image Preview Data Structures)
   n3 = Should Reinitialize Widgets!(execute=@n1.then_1, self, CurData:Array of Movie Graph Image Preview Data Structures=@n2) -> (then, bReinitialize:bool)
   > n4 = Branch(Condition=@n3)
   > Clear!(
    execute
    TargetMap:Map of Movie Graph Render Data Identifier Structures to UI Movie Graph Image Preview Object References=Get CreatedGridWidgets()
   ) -> (then) {is=title K2Node_CallFunction}
   > ClearChildren(self=Get PreviewImageGrid())
   > CreateEnoughGridWidgets!(execute, self, CurData:Array of Movie Graph Image Preview Data Structures=@n2) -> (then) {is=title K2Node_CallFunction}
   UpdatePreviewWidgets!(execute=@n4.else, self, Cur Data:Array of Movie Graph Image Preview Data Structures=@n2) -> (then)
   n5 = Event OnInitializedForPipeline!() -> (OutputDelegate, then, InPipeline:Movie Graph Pipeline Object Reference) {is=title K2Node_Event}
   > Set Pipeline(Pipeline:Movie Graph Pipeline Object Reference=@n5.InPipeline) -> (Output_Get:Movie Graph Pipeline Object Reference)
  ]}
 GetNumRowAndColumnForGrid {nodes [
   n1 = GetNumRowAndColumnForGrid!() -> (then, NumPreviews:int) {is=title K2Node_FunctionEntry}
   > n2 = Switch on Int!(execute:Exec, Selection:int=@n1) -> (Default:Exec, 0:Exec, 1:Exec, 2:Exec, 3:Exec, 4:Exec, 5:Exec, 6:Exec, 7:Exec, 8:Exec, 9:Exec)
   Return Node!(execute=@n2.0, Rows:int=0, Columns:int=0)
   Return Node(execute=@n2.1, Rows=1, Columns=1)
   Return Node(execute=@n2.2, Rows=1, Columns=2)
   Return Node(execute=@n2.3, Rows=2, Columns=2)
   Return Node(execute=@n2.4, Rows=2, Columns=2)
   Return Node(execute=@n2.5, Rows=2, Columns=3)
   Return Node(execute=@n2.6, Rows=2, Columns=3)
   Return Node(execute=@n2.7, Rows=3, Columns=3)
   Return Node(execute=@n2.8, Rows=3, Columns=3)
   Return Node(execute=@n2.9, Rows=3, Columns=3)
   > Return Node(Rows=4, Columns=4)
  ]}
 GetShotEngineWarm_Visibility {nodes [
   GetShotEngineWarm_Visibility() {is=title K2Node_FunctionEntry}
   n1 = GetCurrentSegmentState!(InMovieGraphPipeline:Movie Graph Pipeline Object Reference=Get Pipeline()) -> (ReturnValue:EMovieRenderShotState Enum=Uninitialized)
   > Return Node!(
    execute
    ReturnValue:ESlateVisibility Enum=Select(Option 0=Collapsed, Option 1=Visible, Index="Equal (Enum)"(A:EMovieRenderShotState Enum=@n1, B:EMovieRenderShotState Enum=WarmingUp))
   )
  ]}
 GetShotSampleRow_Visibility {nodes [
   GetShotSampleRow_Visibility() {is=title K2Node_FunctionEntry}
   n1 = GetCurrentSegmentState(InMovieGraphPipeline=Get Pipeline()) -> (ReturnValue=Uninitialized)
   > Return Node(ReturnValue=Select(Option 0=Collapsed, Option 1=Visible, Index="Equal (Enum)"(A=@n1, B=Rendering)))
  ]}
 SetCurrentLocalFrame {nodes [
   SetCurrentLocalFrame() {is=title K2Node_FunctionEntry}
   n1 = Break Movie Pipeline Segment Work Metrics!(MoviePipelineSegmentWorkMetrics:"Movie Pipeline Segment Work Metrics Structure (by ref)"=GetCurrentSegmentWorkMetrics!(InMovieGraphPipeline:Movie Graph Pipeline Object Reference=Get Pipeline()) -> (ReturnValue:Movie Pipeline Segment Work Metrics Structure)) -> (
    SegmentName:String
    OutputFrameIndex:int
    TotalOutputFrameCount:int
    OutputSubSampleIndex:int
    TotalSubSampleCount:int
    EngineWarmUpFrameIndex:int
    TotalEngineWarmUpFrameCount:int
   ) {is=title K2Node_BreakStruct}
   > Set CurrentFrameNumber(CurrentFrameNumber:int=int + int(A=@n1.OutputFrameIndex, B=1) -> (ReturnValue=0) {is=title K2Node_CommutativeAssociativeBinaryOperator}) -> (Output_Get:int=0)
   > Set TotalFrameNumber(TotalFrameNumber:int=@n1.TotalOutputFrameCount) -> (Output_Get:int=0)
   n2 = Get TotalFrameNumber() -> (TotalFrameNumber:int=0)
   n3 = "To Float (Integer)"(InInt=SelectInt(A=Get TotalFrameNumber() -> (TotalFrameNumber=0), B=1, bPickA="integer > integer"(A=@n2)))
   n4 = "To Float (Integer)"(InInt=Get CurrentFrameNumber() -> (CurrentFrameNumber:int=0))
   n5 = float * float(A="Clamp (Float)"(Value=float / float(A=@n4, B=@n3) {is=title K2Node_CallFunction}, Min=0, Max=1), B=100)
   n6 = Get TotalFrameNumber() -> (TotalFrameNumber=0)
   > SetValueText!(
    execute
    self:UI Movie Render Pipeline Info Table Row Object Reference=Get ShotFrameCountRow() -> (ShotFrameCountRow:UI Movie Render Pipeline Info Table Row Object Reference)
    Text:Text=Format Text!(
     Format="LOCTEXT(\"{CurOutput} of {TotalOutput} ({Percent}%)\")"
     CurOutput:int=Get CurrentFrameNumber() -> (CurrentFrameNumber=0)
     TotalOutput:int="Clamp (Integer)"(Value=@n6, Min=0, Max=@n6) -> (ReturnValue=0)
     Percent:Text="To Text (Float)"(Value=@n5, MinimumIntegralDigits=3, MaximumIntegralDigits=3, MinimumFractionalDigits=0, MaximumFractionalDigits=0)
    ) -> (Result)
   ) -> (then)
   > Return Node!(execute)
  ]}
 SetCurrentShotName {nodes [
   SetCurrentShotName() {is=title K2Node_FunctionEntry}
   n1 = GetCurrentSegmentName!(InMovieGraphPipeline:Movie Graph Pipeline Object Reference=Get Pipeline()) -> (OutOuterName:Text, OutInnerName:Text)
   > n2 = Branch(Condition=TextIsEmpty(InText=@n1.OutInnerName))
   n3 = Get Pipeline()
   n4 = GetOverallSegmentCounts!(InMovieGraphPipeline:Movie Graph Pipeline Object Reference=Get Pipeline()) -> (OutCurrentIndex:int=0, OutTotalCount:int=0)
   > SetValueText(
    self=Get ShotNameRow() -> (ShotNameRow:UI Movie Render Pipeline Info Table Row Object Reference)
    Text=Format Text!(
     Format="LOCTEXT(\"{OuterName} ({Num} of {Total})\")"
     OuterName:Text=GetCurrentSegmentName(InMovieGraphPipeline=@n3)
     Num:int=int + int(A=@n4, B=1) -> (ReturnValue=0)
     Total:int=@n4.OutTotalCount
    ) -> (Result)
   )
   > Return Node()
   n5 = GetOverallSegmentCounts(InMovieGraphPipeline=Get Pipeline()) -> (OutCurrentIndex=0, OutTotalCount=0)
   SetValueText(
    execute=@n2.else
    self=Get ShotNameRow()
    Text=Format Text!(
     Format="LOCTEXT(\"{OuterName} ({InnerName}) ({Num} of {Total})\")"
     OuterName=@n1
     InnerName:Text=@n1.OutInnerName
     Num=int + int(A=@n5, B=1) -> (ReturnValue=0)
     Total=@n5.OutTotalCount
    ) -> (Result)
   )
   > Return Node()
  ]}
 SetElapsedTime {nodes [
   SetElapsedTime() {is=title K2Node_FunctionEntry}
   n1 = Get Pipeline()
   n2 = BreakTimespan!(InTimespan:Timespan Structure=DateTime - DateTime!(
     A:Date Time Structure=UTC Now!() -> (ReturnValue:Date Time Structure)
     B:Date Time Structure=GetJobInitializationTime!(InMovieGraphPipeline:Movie Graph Pipeline Object Reference=@n1) -> (ReturnValue:Date Time Structure)
    ) -> (ReturnValue:Timespan Structure)) -> (Days:int=0, Hours:int=0, Minutes:int=0, Seconds:int=0, Milliseconds:int=0)
   > SetValueText(
    self=Get ElapsedTimeRow() -> (ElapsedTimeRow:UI Movie Render Pipeline Info Table Row Object Reference)
    Text=Format Text!(
     Format="LOCTEXT(\"[{Hour}:{Minute}:{Second}]\")"
     Hour:Text="To Text (Integer)"(Value=@n2.Hours, MinimumIntegralDigits=2, MaximumIntegralDigits=2)
     Minute:Text="To Text (Integer)"(Value=@n2.Minutes, MinimumIntegralDigits=2, MaximumIntegralDigits=2)
     Second:Text="To Text (Integer)"(Value=@n2.Seconds, MinimumIntegralDigits=2, MaximumIntegralDigits=2)
    ) -> (Result)
   )
   > Return Node()
   "To Text (Integer)"(Value=@n2.Milliseconds, MinimumIntegralDigits=3, MaximumIntegralDigits=3)
  ]}
 SetEstimatedTime {nodes [
   SetEstimatedTime() {is=title K2Node_FunctionEntry}
   n1 = GetEstimatedTimeRemaining!(InMovieGraphPipeline:Movie Graph Pipeline Object Reference=Get Pipeline()) -> (OutEstimate:Timespan Structure, ReturnValue:bool)
   > n2 = Branch(Condition=@n1.ReturnValue)
   n3 = BreakTimespan(InTimespan=@n1) -> (Days=0, Hours=0, Minutes=0, Seconds=0, Milliseconds=0)
   > SetValueText(
    self=Get EstimatedTimeRemainingRow() -> (EstimatedTimeRemainingRow:UI Movie Render Pipeline Info Table Row Object Reference)
    Text=Format Text(
     Format="LOCTEXT(\"[{Hour}:{Minute}:{Second}]\")"
     Hour="To Text (Integer)"(Value=@n3.Hours, MinimumIntegralDigits=2, MaximumIntegralDigits=2)
     Minute="To Text (Integer)"(Value=@n3.Minutes, MinimumIntegralDigits=2, MaximumIntegralDigits=2)
     Second="To Text (Integer)"(Value=@n3.Seconds, MinimumIntegralDigits=2, MaximumIntegralDigits=2)
    )
   )
   SetValueText(execute=@n2.else, self=Get EstimatedTimeRemainingRow(), Text=MakeLiteralText(Value="LOCTEXT(\"Calculating\")"))
   "To Text (Integer)"(Value=@n3.Milliseconds, MinimumIntegralDigits=3, MaximumIntegralDigits=3)
  ]}
 SetPipelineState_Color {nodes [
   SetPipelineState_Color() {is=title K2Node_FunctionEntry}
   n1 = Get Pipeline()
   > SetValueColor!(
    execute
    self:UI Movie Render Pipeline Info Table Row Object Reference=Get PipelineStateRow() -> (PipelineStateRow:UI Movie Render Pipeline Info Table Row Object Reference)
    Color:Slate Color Structure=Select!(
     Uninitialized:Slate Color Structure=Make Slate Color(SpecifiedColor=(R=1,G=0,B=0.0574,A=1))
     ProducingFrames:Slate Color Structure=Make Slate Color(SpecifiedColor=(R=0.4912,G=1,B=0.1418,A=1))
     Finalize:Slate Color Structure=Make Slate Color(SpecifiedColor=(R=0,G=0.1061,B=1,A=1))
     Export:Slate Color Structure=Make Slate Color(SpecifiedColor=(R=1,G=0.5257,B=0,A=1))
     Finished:Slate Color Structure
     Index:EMovieRenderPipelineState Enum=GetPipelineState!(self:Movie Pipeline Base Object Reference=@n1) -> (ReturnValue:EMovieRenderPipelineState Enum=Uninitialized)
    ) -> (ReturnValue:Slate Color Structure)
   ) -> (then)
   > Return Node()
  ]}
 SetPipelineState_Text {nodes [
   SetPipelineState_Text() {is=title K2Node_FunctionEntry}
   n1 = GetPipelineState!(InMovieGraphPipeline:Movie Graph Pipeline Object Reference=Get Pipeline()) -> (ReturnValue=Uninitialized)
   > SetValueText(
    self=Get PipelineStateRow()
    Text="To Text (String)"(InString=Enum to String!(Enumerator:EMovieRenderPipelineState Enum=@n1) -> (ReturnValue:String))
   )
   > Return Node()
  ]}
 SetSequenceName {nodes [
   SetSequenceName() {is=title K2Node_FunctionEntry}
   > SetValueText(
    self=Get CurrentSequenceRow() -> (CurrentSequenceRow:UI Movie Render Pipeline Info Table Row Object Reference)
    Text=GetJobName!(InMovieGraphPipeline:Movie Graph Pipeline Object Reference=Get Pipeline()) -> (ReturnValue:Text)
   )
   > Return Node()
  ]}
 SetShotState_Color {nodes [
   SetShotState_Color() {is=title K2Node_FunctionEntry}
   n1 = Get Pipeline()
   > SetValueColor(
    self=Get ShotStateRow() -> (ShotStateRow:UI Movie Render Pipeline Info Table Row Object Reference)
    Color=Select!(
     Uninitialized=Make Slate Color(SpecifiedColor=(R=1,G=0,B=0.0186,A=1))
     WarmingUp:Slate Color Structure=Make Slate Color(SpecifiedColor=(R=0,G=0.4707,B=1,A=1))
     MotionBlur:Slate Color Structure=Make Slate Color(SpecifiedColor=(R=0.3966,G=0.0528,B=1,A=1))
     Rendering:Slate Color Structure=Make Slate Color(SpecifiedColor=(R=0.434,G=1,B=0.2091,A=1))
     CoolingDown:Slate Color Structure=Make Slate Color(SpecifiedColor=(R=0,G=0.4707,B=1,A=1))
     Finished=Make Slate Color(SpecifiedColor=(R=0.8345,G=0,B=1,A=1))
     Index:EMovieRenderShotState Enum=GetCurrentSegmentState(InMovieGraphPipeline=@n1) -> (ReturnValue=Uninitialized)
    ) -> (ReturnValue)
   )
   > Return Node()
   Make Slate Color(SpecifiedColor=(R=1,G=0.3006,B=0,A=1))
  ]}
 SetShotState_Text {nodes [
   SetShotState_Text() {is=title K2Node_FunctionEntry}
   n1 = GetCurrentSegmentState(InMovieGraphPipeline=Get Pipeline()) -> (ReturnValue=Uninitialized)
   > SetValueText(self=Get ShotStateRow(), Text="To Text (String)"(InString=Enum to String(Enumerator:EMovieRenderShotState Enum=@n1)))
   > Return Node()
  ]}
 SetSubSample {nodes [
   SetSubSample() {is=title K2Node_FunctionEntry}
   n1 = Break Movie Pipeline Segment Work Metrics(MoviePipelineSegmentWorkMetrics=GetCurrentSegmentWorkMetrics(InMovieGraphPipeline=Get Pipeline()))
   n2 = SelectInt(A=2, B=1, bPickA="Equal (Integer)"(A=@n1.OutputSubSampleIndex, B=-1))
   > SetValueText(
    self=Get ShotSubFrameRow() -> (ShotSubFrameRow:UI Movie Render Pipeline Info Table Row Object Reference)
    Text=Format Text!(
     Format="LOCTEXT(\"{CurFrame} of {Total}\")"
     CurFrame:int=int + int(A=@n1.OutputSubSampleIndex, B=@n2) -> (ReturnValue=0)
     Total:int=@n1.TotalSubSampleCount
    ) -> (Result)
   )
   > Return Node()
  ]}
 SetTotalFrames {nodes [
   SetTotalFrames() {is=title K2Node_FunctionEntry}
   n1 = GetOverallOutputFrames!(InMovieGraphPipeline:Movie Graph Pipeline Object Reference=Get Pipeline()) -> (OutCurrentIndex:int=0, OutTotalCount:int=0)
   > Set CurrentFrameNumber(CurrentFrameNumber=@n1) -> (Output_Get=0)
   > Set TotalFrameNumber(TotalFrameNumber=@n1.OutTotalCount) -> (Output_Get=0)
   n2 = Get TotalFrameNumber() -> (TotalFrameNumber=0)
   n3 = "To Float (Integer)"(InInt=SelectInt(A=Get TotalFrameNumber() -> (TotalFrameNumber=0), B=1, bPickA="integer > integer"(A=@n2)))
   n4 = "To Float (Integer)"(InInt=Get CurrentFrameNumber() -> (CurrentFrameNumber=0))
   n5 = float * float(A="Clamp (Float)"(Value=float / float(A=@n4, B=@n3) {is=title K2Node_PromotableOperator}, Min=0, Max=1), B=100)
   n6 = Get TotalFrameNumber() -> (TotalFrameNumber=0)
   > SetValueText(
    self=Get TotalFrameRow() -> (TotalFrameRow:UI Movie Render Pipeline Info Table Row Object Reference)
    Text=Format Text!(
     Format="LOCTEXT(\"{CurOutput} of {TotalOutput} ({Percent}%)\")"
     CurOutput:int=Get CurrentFrameNumber() -> (CurrentFrameNumber=0)
     TotalOutput:int="Clamp (Integer)"(Value=@n6, Min=0, Max=@n6) -> (ReturnValue=0)
     Percent:Text="To Text (Float)"(Value=@n5, MinimumIntegralDigits=3, MaximumIntegralDigits=3, MinimumFractionalDigits=0, MaximumFractionalDigits=0)
    ) -> (Result)
   )
   > Return Node()
  ]}
 SetWarmUpFrames {nodes [
   SetWarmUpFrames() {is=title K2Node_FunctionEntry}
   > n1 = Sequence()
   n2 = Break Movie Pipeline Segment Work Metrics(MoviePipelineSegmentWorkMetrics=GetCurrentSegmentWorkMetrics(InMovieGraphPipeline=Get Pipeline()))
   n3 = int + int(A=@n2.EngineWarmUpFrameIndex, B=1) -> (ReturnValue=0)
   > SetValueText(
    self=Get ShotEngineWarmUpFramesRow() -> (ShotEngineWarmUpFramesRow:UI Movie Render Pipeline Info Table Row Object Reference)
    Text=Format Text!(
     Format="LOCTEXT(\"{Current} of {Total}\")"
     Current:Text="To Text (Integer)"(Value=@n3)
     Total:Text="To Text (Integer)"(Value=@n2.TotalEngineWarmUpFrameCount)
    ) -> (Result)
   )
   Return Node(execute=@n1.then_2)
  ]}
 Should Reinitialize Widgets {nodes [
   n1 = Should Reinitialize Widgets!() -> (then, CurData:Array of Movie Graph Image Preview Data Structures) {is=title K2Node_FunctionEntry}
   n2 = Get CreatedGridWidgets()
   > n3 = Branch(Condition="Equal (Integer)"(
     A=Length(TargetArray=@n1)
     B=Length!(TargetMap:Map of Movie Graph Render Data Identifier Structures to UI Movie Graph Image Preview Object References=@n2) -> (ReturnValue:int=0) {is=title K2Node_CallFunction}
    ))
   > Set bChanged(bChanged:bool=false) -> (Output_Get:bool)
   > n4 = For Each Loop with Break!(Exec:Exec, Array:Array of Movie Graph Image Preview Data Structures=@n1, Break:Exec=@n6.then) -> (LoopBody:Exec, Array Element:Movie Graph Image Preview Data Structure, Array Index:int, Completed:Exec)
   Return Node!(execute=@n4.Completed, bReinitialize:bool=Get bChanged() -> (bChanged:bool))
   > n5 = Branch(Condition=Contains!(
     TargetMap:Map of Movie Graph Render Data Identifier Structures to UI Movie Graph Image Preview Object References=Get CreatedGridWidgets()
     Key:"Movie Graph Render Data Identifier Structure (by ref)"=Break Movie Graph Image Preview Data(MovieGraphImagePreviewData=@n4) -> (bMultipleCameraNames=~).Identifier
    ) -> (ReturnValue))
   n6 = Set bChanged(execute=@n5.else, bChanged=true)
   Return Node(execute=@n3.else, bReinitialize=true)
  ]}
 UpdatePreviewWidgets {nodes [
   n1 = UpdatePreviewWidgets!() -> (then, Cur Data:Array of Movie Graph Image Preview Data Structures) {is=title K2Node_FunctionEntry}
   > n2 = For Each Loop(Array=@n1)
   n3 = Break Movie Graph Image Preview Data(MovieGraphImagePreviewData=@n2) -> (bMultipleCameraNames=~)
   n4 = Find(
    TargetMap:Map of Movie Graph Render Data Identifier Structures to UI Movie Graph Image Preview Object References=Get CreatedGridWidgets()
    Key:"Movie Graph Render Data Identifier Structure (by ref)"=@n3.Identifier
   ) -> (Value:UI Movie Graph Image Preview Object Reference)
   > Branch(Condition=@n4.ReturnValue)
   > n5 = Branch(Condition="Equal (Object)"(A=@n3, B=Get TextureRepresented!(self:UI Movie Graph Image Preview Object Reference=@n4) -> (TextureRepresented:Texture Object Reference)))
   SetPreviewImage(execute=@n5.else, self=@n4, Data=@n2)
  ]}
 UpdateRowWidgetVisibilities {nodes [
   UpdateRowWidgetVisibilities() {is=title K2Node_FunctionEntry}
   > SetVisibility(
    self=Get ShotSubFrameRow()
    InVisibility=GetShotSampleRow_Visibility!() -> (ReturnValue:ESlateVisibility Enum=Visible) {is=title K2Node_CallFunction}
   )
   > SetVisibility(
    self=Get ShotEngineWarmUpFramesRow()
    InVisibility=GetShotEngineWarm_Visibility!() -> (ReturnValue:ESlateVisibility Enum=Visible) {is=title K2Node_CallFunction}
   )
  ]}
]
variables [
 CreatedGridWidgets {default {}}
 Pipeline {type=MovieGraphPipeline, default=null}
 PreviewMaterialInst {type=MaterialInstanceDynamic}
]
widgets {
 info {
  inheritedWidgetCount=0
  parentClass=/Script/MovieRenderPipelineCore.MovieGraphRenderPreviewWidget
  rootWidgetClass=&WidgetSwitcher
  widgetCount=24
 }
 widgets [
  WidgetSwitcher_0 {widgetClassPath=&WidgetSwitcher, bIsVariable=true, widget=WidgetSwitcher_0, properties {activeWidgetIndex=1}}
  VerticalBox_14 {widgetClassPath=&VerticalBox, parent=WidgetSwitcher_0, widget=VerticalBox_14, slot {class=WidgetSwitcherSlot}}
  VerticalBox_121 {
   widgetClassPath=&VerticalBox
   parent=VerticalBox_14
   widget=VerticalBox_121
   slot {class=VerticalBoxSlot, properties {size {sizeRule=Fill, value=1}}}
  }
  PreviewImageGrid {
   widgetClassPath=/Script/UMG.UniformGridPanel
   bIsVariable=true
   parent=VerticalBox_121
   widget=PreviewImageGrid
   slot {class=VerticalBoxSlot, properties {padding=<ltrb 0 4 0 0>, size {sizeRule=Fill, value=1}}}
  }
  TextBlock {
   widgetClassPath=&TextBlock
   parent=VerticalBox_121
   widget=TextBlock
   properties {
    text=Render Preview
    font {
     bForceMonospaced=false
     bMaterialIsStencil=false
     fontMaterial=null
     fontObject=&Roboto
     letterSpacing=0
     monospacedWidth=1
     size=48
     skewAmount=0
     typefaceFontName=Bold
     outlineSettings {
      bApplyOutlineToDropShadows=false
      bMiteredCorners=false
      bSeparateFillAlpha=false
      outlineMaterial=null
      outlineSize=0
      outlineColor=<rgba 0 0 0 1>
     }
    }
    shadowColorAndOpacity=<rgba 0 0 0 1>
   }
   slot {class=VerticalBoxSlot, properties {horizontalAlignment=HAlign_Center, verticalAlignment=VAlign_Center, padding=<ltrb 0 10 0 0>}}
  }
  VerticalBox_160 {
   widgetClassPath=&VerticalBox
   parent=VerticalBox_14
   widget=VerticalBox_160
   slot {class=VerticalBoxSlot, properties {verticalAlignment=VAlign_Bottom}}
  }
  Overlay_70 {
   widgetClassPath=&Overlay
   parent=VerticalBox_160
   widget=Overlay_70
   slot {class=VerticalBoxSlot, properties {verticalAlignment=VAlign_Top}}
  }
  Image_124 {
   widgetClassPath=&Image
   parent=Overlay_70
   widget=Image_124
   properties {colorAndOpacity=<rgba 0 0 0 0.25>}
   slot {class=OverlaySlot, properties {horizontalAlignment=HAlign_Fill, verticalAlignment=VAlign_Fill}}
  }
  HorizontalBox {
   widgetClassPath=&HorizontalBox
   parent=Overlay_70
   widget=HorizontalBox
   slot {class=OverlaySlot, properties {horizontalAlignment=HAlign_Fill, verticalAlignment=VAlign_Fill, padding=<ltrb 12 12 12 12>}}
  }
  VerticalBox_158 {
   widgetClassPath=&VerticalBox
   parent=HorizontalBox
   widget=VerticalBox_158
   slot {class=HorizontalBoxSlot, properties {horizontalAlignment=HAlign_Left, size {sizeRule=Fill, value=1}}}
  }
  CurrentSequenceRow {
   widgetClassPath=&/UI_MovieRenderPipelineInfoTableRow
   bIsVariable=true
   parent=VerticalBox_158
   widget=CurrentSequenceRow
   properties {leftColumnText=Current Sequence:, rightColumnText=Pipeline Test Sequence}
   slot {class=VerticalBoxSlot}
  }
  TotalFrameRow {
   widgetClassPath=&UI_MovieRenderPipelineInfoTableRow
   bIsVariable=true
   parent=VerticalBox_158
   widget=TotalFrameRow
   properties {leftColumnText=Total Frames:, rightColumnText="1245 of 6000 (20%)"}
   slot {class=VerticalBoxSlot}
  }
  ElapsedTimeRow {
   widgetClassPath=&UI_MovieRenderPipelineInfoTableRow
   bIsVariable=true
   parent=VerticalBox_158
   widget=ElapsedTimeRow
   properties {leftColumnText=Elapsed Time:, rightColumnText="[00:06:34]"}
   slot {class=VerticalBoxSlot}
  }
  EstimatedTimeRemainingRow {
   widgetClassPath=&UI_MovieRenderPipelineInfoTableRow
   bIsVariable=true
   parent=VerticalBox_158
   widget=EstimatedTimeRemainingRow
   properties {leftColumnText=Est. Time Remaining:, rightColumnText="[00:12:34]"}
   slot {class=VerticalBoxSlot}
  }
  PipelineStateRow {
   widgetClassPath=&UI_MovieRenderPipelineInfoTableRow
   bIsVariable=true
   parent=VerticalBox_158
   widget=PipelineStateRow
   properties {leftColumnText=Pipeline State:, rightColumnText=Producing Frames}
   slot {class=VerticalBoxSlot}
  }
  VerticalBox_44 {
   widgetClassPath=&VerticalBox
   parent=HorizontalBox
   widget=VerticalBox_44
   slot {class=HorizontalBoxSlot, properties {horizontalAlignment=HAlign_Right, size {sizeRule=Fill, value=1}}}
  }
  ShotNameRow {
   widgetClassPath=&UI_MovieRenderPipelineInfoTableRow
   bIsVariable=true
   parent=VerticalBox_44
   widget=ShotNameRow
   properties {leftColumnText=Cut Name:, rightColumnText="02_0400 (1 of 5)"}
   slot {class=VerticalBoxSlot}
  }
  ShotFrameCountRow {
   widgetClassPath=&UI_MovieRenderPipelineInfoTableRow
   bIsVariable=true
   parent=VerticalBox_44
   widget=ShotFrameCountRow
   properties {leftColumnText=Frame:, rightColumnText="1230 of 3000 (20%)"}
   slot {class=VerticalBoxSlot}
  }
  ShotSubFrameRow {
   widgetClassPath=&UI_MovieRenderPipelineInfoTableRow
   bIsVariable=true
   parent=VerticalBox_44
   widget=ShotSubFrameRow
   properties {leftColumnText=Sub Sample:, rightColumnText=32 of 84}
   slot {class=VerticalBoxSlot}
  }
  ShotEngineWarmUpFramesRow {
   widgetClassPath=&UI_MovieRenderPipelineInfoTableRow
   bIsVariable=true
   parent=VerticalBox_44
   widget=ShotEngineWarmUpFramesRow
   properties {leftColumnText=Engine Warm Up Frames:, rightColumnText=1/4}
   slot {class=VerticalBoxSlot}
  }
  ShotHandleFramesRow {
   widgetClassPath=&UI_MovieRenderPipelineInfoTableRow
   bIsVariable=true
   parent=VerticalBox_44
   widget=ShotHandleFramesRow
   properties {leftColumnText=Handle Frames:, rightColumnText=4 of 12, visibility=Collapsed}
   slot {class=VerticalBoxSlot}
  }
  ShotStateRow {
   widgetClassPath=&UI_MovieRenderPipelineInfoTableRow
   bIsVariable=true
   parent=VerticalBox_44
   widget=ShotStateRow
   properties {leftColumnText=Cut State:, rightColumnText=Uninitialized}
   slot {class=VerticalBoxSlot}
  }
  VerticalBox_62 {widgetClassPath=&VerticalBox, parent=WidgetSwitcher_0, widget=VerticalBox_62, slot {class=WidgetSwitcherSlot}}
  TextBlock_1 {
   widgetClassPath=&TextBlock
   parent=VerticalBox_62
   widget=TextBlock_1
   properties {
    text=Initializing
    font {
     bForceMonospaced=false
     bMaterialIsStencil=false
     fontMaterial=null
     fontObject=&Roboto
     letterSpacing=0
     monospacedWidth=1
     size=48
     skewAmount=0
     typefaceFontName=Bold
     outlineSettings {
      bApplyOutlineToDropShadows=false
      bMiteredCorners=false
      bSeparateFillAlpha=false
      outlineMaterial=null
      outlineSize=0
      outlineColor=<rgba 0 0 0 1>
     }
    }
    shadowColorAndOpacity=<rgba 0 0 0 1>
   }
   slot {
    class=VerticalBoxSlot
    properties {horizontalAlignment=HAlign_Center, verticalAlignment=VAlign_Bottom, padding=<ltrb 0 0 0 64>, size {sizeRule=Fill, value=1}}
   }
  }
 ]
}
