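# Screen overlay widget derived from MovieRenderDebugWidget that shows a low quality preview and progress of a Movie Render Pipeline job.
# OnInitializedForPipeline stores the pipeline and puts a dynamic preview material on the hidden image.
# Tick switches from the Initializing page and polls the pipeline to fill overall and camera cut rows, color states and toggle warm-up and sub-sample rows.
# Tick also feeds the preview texture to the material, sizing and revealing the image.
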WidgetBlueprint /MovieRenderPipeline/Blueprints/UI_MovieRenderPipelineScreenOverlay
graphs [
 EventGraph {nodes [
   Event Tick()
   > SetActiveWidgetIndex(self=Get WidgetSwitcher_0() -> (WidgetSwitcher_0:Widget Switcher Object Reference), Index=0)
   > n1 = Sequence!(execute) -> (then_0, then_1, then_2:Exec)
   > SetSequenceName()
   > SetTotalFrames()
   > SetElapsedTime()
   > SetEstimatedTime()
   > SetPipelineState_Text()
   > SetPipelineState_Color()
   > SetCurrentShotName()
   > SetCurrentLocalFrame()
   > SetSubSample()
   > SetWarmUpFrames()
   > SetShotState_Text()
   > SetShotState_Color()
   > UpdateRowWidgetVisibilities()
   n2 = GetPreviewTexture!(self:Movie Pipeline Object Reference=Get Pipeline() -> (Pipeline:Movie Pipeline Object Reference)) -> (ReturnValue:Texture Object Reference)
   SetTextureParameterValue(
    execute=@n1.then_1
    self=Get PreviewMaterialInst() -> (PreviewMaterialInst:Material Instance Dynamic Object Reference)
    ParameterName=RenderTexture
    Value=@n2
   )
   > n3 = Cast To TextureRenderTarget2D!(execute, Object=@n2) -> (then, CastFailed, AsTexture Render Target 2D:Texture Render Target 2D Object Reference) {is=title K2Node_DynamicCast}
   > SetDesiredSizeOverride!(
    execute
    self:Image Object Reference=Get ScenePreviewImage() -> (ScenePreviewImage:Image Object Reference)
    DesiredSize_X:double="To Float (Integer)"(InInt=Get SizeX!(self:Texture Render Target 2D Object Reference=@n3) -> (SizeX:int=0))
    DesiredSize_Y:double="To Float (Integer)"(InInt=Get SizeY!(self:Texture Render Target 2D Object Reference=@n3) -> (SizeY:int=0))
   ) -> (then)
   > SetVisibility(self=Get ScenePreviewImage(), InVisibility=Visible)
   n4 = Event OnInitializedForPipeline!() -> (OutputDelegate, then, ForPipeline:Movie Pipeline Object Reference) {is=title K2Node_Event}
   > Set Pipeline(Pipeline:Movie Pipeline Object Reference=@n4.ForPipeline) -> (Output_Get:Movie Pipeline Object Reference)
   > n5 = CreateDynamicMaterialInstance!(execute, Parent=&/M_MoviePipelineRenderPreview, OptionalName=None, CreationFlags=None) -> (then, ReturnValue)
   > n6 = Set PreviewMaterialInst(PreviewMaterialInst:Material Instance Dynamic Object Reference=@n5) -> (Output_Get:Material Instance Dynamic Object Reference)
   > SetBrushFromMaterial(self=Get ScenePreviewImage(), Material=@n6)
   > SetVisibility(self=Get ScenePreviewImage(), InVisibility=Hidden)
  ]}
 GetShotEngineWarm_Visibility {nodes [
   GetShotEngineWarm_Visibility() {is=title K2Node_FunctionEntry}
   n1 = GetCurrentSegmentState!(InMoviePipeline:Movie Pipeline Object Reference=Get Pipeline()) -> (ReturnValue:EMovieRenderShotState Enum=Uninitialized)
   > Return Node!(
    execute
    ReturnValue:ESlateVisibility Enum=Select(Option 0=Collapsed, Option 1=Visible, Index="Equal (Enum)"(A:EMovieRenderShotState Enum=@n1, B:EMovieRenderShotState Enum=WarmingUp))
   )
  ]}
 GetShotSampleRow_Visibility {nodes [
   GetShotSampleRow_Visibility() {is=title K2Node_FunctionEntry}
   n1 = GetCurrentSegmentState(InMoviePipeline=Get Pipeline()) -> (ReturnValue=Uninitialized)
   > Return Node(ReturnValue=Select(Option 0=Collapsed, Option 1=Visible, Index="Equal (Enum)"(A=@n1, B=Rendering)))
  ]}
 SetCurrentLocalFrame {nodes [
   SetCurrentLocalFrame() {is=title K2Node_FunctionEntry}
   n1 = Break Movie Pipeline Segment Work Metrics!(MoviePipelineSegmentWorkMetrics:"Movie Pipeline Segment Work Metrics Structure (by ref)"=GetCurrentSegmentWorkMetrics!(InMoviePipeline:Movie Pipeline Object Reference=Get Pipeline()) -> (ReturnValue:Movie Pipeline Segment Work Metrics Structure)) -> (
    SegmentName:String
    OutputFrameIndex:int
    TotalOutputFrameCount:int
    OutputSubSampleIndex:int
    TotalSubSampleCount:int
    EngineWarmUpFrameIndex:int
    TotalEngineWarmUpFrameCount:int
   ) {is=title K2Node_BreakStruct}
   > Set CurrentFrameNumber(CurrentFrameNumber:int=int + int(A=@n1.OutputFrameIndex, B=1) -> (ReturnValue=0) {is=title K2Node_CommutativeAssociativeBinaryOperator}) -> (Output_Get:int=0)
   > Set TotalFrameNumber(TotalFrameNumber:int=@n1.TotalOutputFrameCount) -> (Output_Get:int=0)
   n2 = "To Float (Integer)"(InInt=Get TotalFrameNumber() -> (TotalFrameNumber:int=0))
   n3 = "To Float (Integer)"(InInt=Get CurrentFrameNumber() -> (CurrentFrameNumber:int=0))
   n4 = float * float(A="Clamp (Float)"(Value=float / float(A=@n3, B=@n2) {is=title K2Node_CallFunction}, Min=0, Max=1), B=100)
   n5 = Get TotalFrameNumber() -> (TotalFrameNumber=0)
   > SetValueText!(
    execute
    self:UI Movie Render Pipeline Info Table Row Object Reference=Get ShotFrameCountRow() -> (ShotFrameCountRow:UI Movie Render Pipeline Info Table Row Object Reference)
    Text:Text=Format Text!(
     Format="LOCTEXT(\"{CurOutput} of {TotalOutput} ({Percent}%)\")"
     CurOutput:int=Get CurrentFrameNumber() -> (CurrentFrameNumber=0)
     TotalOutput:int="Clamp (Integer)"(Value=@n5, Min=0, Max=@n5) -> (ReturnValue=0)
     Percent:Text="To Text (Float)"(Value=@n4, MinimumIntegralDigits=3, MaximumIntegralDigits=3, MinimumFractionalDigits=0, MaximumFractionalDigits=0)
    ) -> (Result)
   ) -> (then)
   > Return Node!(execute)
  ]}
 SetCurrentShotName {nodes [
   SetCurrentShotName() {is=title K2Node_FunctionEntry}
   n1 = Get Pipeline()
   n2 = GetCurrentSegmentName!(InMoviePipeline:Movie Pipeline Object Reference=@n1) -> (OutOuterName:Text, OutInnerName:Text)
   > n3 = Branch(Condition=TextIsEmpty(InText=@n2.OutInnerName))
   n4 = GetOverallSegmentCounts!(InMoviePipeline:Movie Pipeline Object Reference=@n1) -> (OutCurrentIndex:int=0, OutTotalCount:int=0)
   > SetValueText(
    self=Get ShotNameRow() -> (ShotNameRow:UI Movie Render Pipeline Info Table Row Object Reference)
    Text=Format Text!(
     Format="LOCTEXT(\"{OuterName} ({Num} of {Total})\")"
     OuterName:Text=GetCurrentSegmentName(InMoviePipeline=@n1)
     Num:int=int + int(A=@n4, B=1) -> (ReturnValue=0)
     Total:int=@n4.OutTotalCount
    ) -> (Result)
   )
   > Return Node()
   n5 = GetOverallSegmentCounts(InMoviePipeline=@n1) -> (OutCurrentIndex=0, OutTotalCount=0)
   SetValueText(
    execute=@n3.else
    self=Get ShotNameRow()
    Text=Format Text!(
     Format="LOCTEXT(\"{OuterName} ({InnerName}) ({Num} of {Total})\")"
     OuterName=@n2
     InnerName:Text=@n2.OutInnerName
     Num=int + int(A=@n5, B=1) -> (ReturnValue=0)
     Total=@n5.OutTotalCount
    ) -> (Result)
   )
   > Return Node()
  ]}
 SetElapsedTime {nodes [
   SetElapsedTime() {is=title K2Node_FunctionEntry}
   n1 = Get Pipeline()
   n2 = BreakTimespan!(InTimespan:Timespan Structure=DateTime - DateTime!(
     A:Date Time Structure=UTC Now!() -> (ReturnValue:Date Time Structure)
     B:Date Time Structure=GetJobInitializationTime!(InMoviePipeline:Movie Pipeline Object Reference=@n1) -> (ReturnValue:Date Time Structure)
    ) -> (ReturnValue:Timespan Structure)) -> (Days:int=0, Hours:int=0, Minutes:int=0, Seconds:int=0, Milliseconds:int=0)
   > SetValueText(
    self=Get ElapsedTimeRow() -> (ElapsedTimeRow:UI Movie Render Pipeline Info Table Row Object Reference)
    Text=Format Text!(
     Format="LOCTEXT(\"[{Hour}:{Minute}:{Second}]\")"
     Hour:Text="To Text (Integer)"(Value=@n2.Hours, MinimumIntegralDigits=2, MaximumIntegralDigits=2)
     Minute:Text="To Text (Integer)"(Value=@n2.Minutes, MinimumIntegralDigits=2, MaximumIntegralDigits=2)
     Second:Text="To Text (Integer)"(Value=@n2.Seconds, MinimumIntegralDigits=2, MaximumIntegralDigits=2)
    ) -> (Result)
   )
   > Return Node()
   "To Text (Integer)"(Value=@n2.Milliseconds, MinimumIntegralDigits=3, MaximumIntegralDigits=3)
  ]}
 SetEstimatedTime {nodes [
   SetEstimatedTime() {is=title K2Node_FunctionEntry}
   n1 = GetEstimatedTimeRemaining!(InPipeline:Movie Pipeline Object Reference=Get Pipeline()) -> (OutEstimate:Timespan Structure, ReturnValue:bool)
   > n2 = Branch(Condition=@n1.ReturnValue)
   n3 = BreakTimespan(InTimespan=@n1) -> (Days=0, Hours=0, Minutes=0, Seconds=0, Milliseconds=0)
   > SetValueText(
    self=Get EstimatedTimeRemainingRow() -> (EstimatedTimeRemainingRow:UI Movie Render Pipeline Info Table Row Object Reference)
    Text=Format Text(
     Format="LOCTEXT(\"[{Hour}:{Minute}:{Second}]\")"
     Hour="To Text (Integer)"(Value=@n3.Hours, MinimumIntegralDigits=2, MaximumIntegralDigits=2)
     Minute="To Text (Integer)"(Value=@n3.Minutes, MinimumIntegralDigits=2, MaximumIntegralDigits=2)
     Second="To Text (Integer)"(Value=@n3.Seconds, MinimumIntegralDigits=2, MaximumIntegralDigits=2)
    )
   )
   SetValueText(execute=@n2.else, self=Get EstimatedTimeRemainingRow(), Text=MakeLiteralText(Value="LOCTEXT(\"Calculating\")"))
   "To Text (Integer)"(Value=@n3.Milliseconds, MinimumIntegralDigits=3, MaximumIntegralDigits=3)
  ]}
 SetPipelineState_Color {nodes [
   SetPipelineState_Color() {is=title K2Node_FunctionEntry}
   n1 = Get Pipeline()
   > SetValueColor!(
    execute
    self:UI Movie Render Pipeline Info Table Row Object Reference=Get PipelineStateRow() -> (PipelineStateRow:UI Movie Render Pipeline Info Table Row Object Reference)
    Color:Slate Color Structure=Select!(
     Uninitialized:Slate Color Structure=Make Slate Color(SpecifiedColor=(R=1,G=0,B=0.0574,A=1))
     ProducingFrames:Slate Color Structure=Make Slate Color(SpecifiedColor=(R=0.4912,G=1,B=0.1418,A=1))
     Finalize:Slate Color Structure=Make Slate Color(SpecifiedColor=(R=0,G=0.1061,B=1,A=1))
     Export:Slate Color Structure=Make Slate Color(SpecifiedColor=(R=1,G=0.5257,B=0,A=1))
     Finished:Slate Color Structure
     Index:EMovieRenderPipelineState Enum=GetPipelineState!(InPipeline:Movie Pipeline Object Reference=@n1) -> (ReturnValue:EMovieRenderPipelineState Enum=Uninitialized)
    ) -> (ReturnValue:Slate Color Structure)
   ) -> (then)
   > Return Node()
  ]}
 SetPipelineState_Text {nodes [
   SetPipelineState_Text() {is=title K2Node_FunctionEntry}
   n1 = GetPipelineState(InPipeline=Get Pipeline()) -> (ReturnValue=Uninitialized)
   > SetValueText(
    self=Get PipelineStateRow()
    Text="To Text (String)"(InString=Enum to String!(Enumerator:EMovieRenderPipelineState Enum=@n1) -> (ReturnValue:String))
   )
   > Return Node()
  ]}
 SetSequenceName {nodes [
   SetSequenceName() {is=title K2Node_FunctionEntry}
   > SetValueText(
    self=Get CurrentSequenceRow() -> (CurrentSequenceRow:UI Movie Render Pipeline Info Table Row Object Reference)
    Text=GetJobName!(InMoviePipeline:Movie Pipeline Object Reference=Get Pipeline()) -> (ReturnValue:Text)
   )
   > Return Node()
  ]}
 SetShotState_Color {nodes [
   SetShotState_Color() {is=title K2Node_FunctionEntry}
   n1 = Get Pipeline()
   > SetValueColor(
    self=Get ShotStateRow() -> (ShotStateRow:UI Movie Render Pipeline Info Table Row Object Reference)
    Color=Select!(
     Uninitialized=Make Slate Color(SpecifiedColor=(R=1,G=0,B=0.0186,A=1))
     WarmingUp:Slate Color Structure=Make Slate Color(SpecifiedColor=(R=0,G=0.4707,B=1,A=1))
     MotionBlur:Slate Color Structure=Make Slate Color(SpecifiedColor=(R=0.3966,G=0.0528,B=1,A=1))
     Rendering:Slate Color Structure=Make Slate Color(SpecifiedColor=(R=0.434,G=1,B=0.2091,A=1))
     CoolingDown:Slate Color Structure
     Finished=Make Slate Color(SpecifiedColor=(R=0.8345,G=0,B=1,A=1))
     Index:EMovieRenderShotState Enum=GetCurrentSegmentState(InMoviePipeline=@n1) -> (ReturnValue=Uninitialized)
    ) -> (ReturnValue)
   )
   > Return Node()
   Make Slate Color(SpecifiedColor=(R=1,G=0.3006,B=0,A=1))
  ]}
 SetShotState_Text {nodes [
   SetShotState_Text() {is=title K2Node_FunctionEntry}
   n1 = GetCurrentSegmentState(InMoviePipeline=Get Pipeline()) -> (ReturnValue=Uninitialized)
   > SetValueText(self=Get ShotStateRow(), Text="To Text (String)"(InString=Enum to String(Enumerator:EMovieRenderShotState Enum=@n1)))
   > Return Node()
  ]}
 SetSubSample {nodes [
   SetSubSample() {is=title K2Node_FunctionEntry}
   n1 = Break Movie Pipeline Segment Work Metrics(MoviePipelineSegmentWorkMetrics=GetCurrentSegmentWorkMetrics(InMoviePipeline=Get Pipeline()))
   n2 = SelectInt(A=2, B=1, bPickA="Equal (Integer)"(A=@n1.OutputSubSampleIndex, B=-1))
   > SetValueText(
    self=Get ShotSubFrameRow() -> (ShotSubFrameRow:UI Movie Render Pipeline Info Table Row Object Reference)
    Text=Format Text!(
     Format="LOCTEXT(\"{CurFrame} of {Total}\")"
     CurFrame:int=int + int(A=@n1.OutputSubSampleIndex, B=@n2) -> (ReturnValue=0)
     Total:int=@n1.TotalSubSampleCount
    ) -> (Result)
   )
   > Return Node()
  ]}
 SetTotalFrames {nodes [
   SetTotalFrames() {is=title K2Node_FunctionEntry}
   n1 = GetOverallOutputFrames!(InMoviePipeline:Movie Pipeline Object Reference=Get Pipeline()) -> (OutCurrentIndex:int=0, OutTotalCount:int=0)
   > Set CurrentFrameNumber(CurrentFrameNumber=int + int(A=@n1) {is=title K2Node_PromotableOperator}) -> (Output_Get=0)
   > Set TotalFrameNumber(TotalFrameNumber=@n1.OutTotalCount) -> (Output_Get=0)
   n2 = "To Float (Integer)"(InInt=Get TotalFrameNumber() -> (TotalFrameNumber=0))
   n3 = "To Float (Integer)"(InInt=Get CurrentFrameNumber() -> (CurrentFrameNumber=0))
   n4 = float * float(A="Clamp (Float)"(Value=float / float(A=@n3, B=@n2), Min=0, Max=1), B=100)
   n5 = Get TotalFrameNumber() -> (TotalFrameNumber=0)
   > SetValueText(
    self=Get TotalFrameRow() -> (TotalFrameRow:UI Movie Render Pipeline Info Table Row Object Reference)
    Text=Format Text!(
     Format="LOCTEXT(\"{CurOutput} of {TotalOutput} ({Percent}%)\")"
     CurOutput:int=Get CurrentFrameNumber() -> (CurrentFrameNumber=0)
     TotalOutput:int="Clamp (Integer)"(Value=@n5, Min=0, Max=@n5) -> (ReturnValue=0)
     Percent:Text="To Text (Float)"(Value=@n4, MinimumIntegralDigits=3, MaximumIntegralDigits=3, MinimumFractionalDigits=0, MaximumFractionalDigits=0)
    ) -> (Result)
   )
   > Return Node()
  ]}
 SetWarmUpFrames {nodes [
   SetWarmUpFrames() {is=title K2Node_FunctionEntry}
   > n1 = Sequence()
   n2 = Break Movie Pipeline Segment Work Metrics(MoviePipelineSegmentWorkMetrics=GetCurrentSegmentWorkMetrics(InMoviePipeline=Get Pipeline()))
   n3 = int + int(A=@n2.EngineWarmUpFrameIndex, B=1) -> (ReturnValue=0) {is=title K2Node_CommutativeAssociativeBinaryOperator}
   > SetValueText(
    self=Get ShotEngineWarmUpFramesRow() -> (ShotEngineWarmUpFramesRow:UI Movie Render Pipeline Info Table Row Object Reference)
    Text=Format Text!(
     Format="LOCTEXT(\"{Current} of {Total}\")"
     Current:Text="To Text (Integer)"(Value=@n3)
     Total:Text="To Text (Integer)"(Value=@n2.TotalEngineWarmUpFrameCount)
    ) -> (Result)
   )
   Return Node(execute=@n1.then_2)
  ]}
 UpdateRowWidgetVisibilities {nodes [
   UpdateRowWidgetVisibilities() {is=title K2Node_FunctionEntry}
   > SetVisibility(
    self=Get ShotSubFrameRow()
    InVisibility=GetShotSampleRow_Visibility!() -> (ReturnValue:ESlateVisibility Enum=Visible) {is=title K2Node_CallFunction}
   )
   > SetVisibility(
    self=Get ShotEngineWarmUpFramesRow()
    InVisibility=GetShotEngineWarm_Visibility!() -> (ReturnValue:ESlateVisibility Enum=Visible) {is=title K2Node_CallFunction}
   )
  ]}
]
variables [
 Pipeline {type=MoviePipeline, default=null}
 PreviewMaterialInst {type=MaterialInstanceDynamic}
]
widgets {
 info {
  inheritedWidgetCount=0
  parentClass=/Script/MovieRenderPipelineCore.MovieRenderDebugWidget
  rootWidgetClass=&WidgetSwitcher
  widgetCount=32
 }
 widgets [
  WidgetSwitcher_0 {widgetClassPath=&WidgetSwitcher, bIsVariable=true, widget=WidgetSwitcher_0, properties {activeWidgetIndex=1}}
  VerticalBox_14 {widgetClassPath=&VerticalBox, parent=WidgetSwitcher_0, widget=VerticalBox_14, slot {class=WidgetSwitcherSlot}}
  VerticalBox_121 {
   widgetClassPath=&VerticalBox
   parent=VerticalBox_14
   widget=VerticalBox_121
   slot {class=VerticalBoxSlot, properties {size {sizeRule=Fill, value=1}}}
  }
  ScaleBox_25 {
   widgetClassPath=&ScaleBox
   parent=VerticalBox_121
   widget=ScaleBox_25
   slot {class=VerticalBoxSlot, properties {padding=<ltrb 0 32 0 0>, size {sizeRule=Fill, value=1}}}
  }
  ScenePreviewImage {widgetClassPath=&Image, bIsVariable=true, parent=ScaleBox_25, widget=ScenePreviewImage, slot {class=ScaleBoxSlot}}
  TextBlock {
   widgetClassPath=&TextBlock
   parent=VerticalBox_121
   widget=TextBlock
   properties {
    text=Render Preview
    font {
     bForceMonospaced=false
     bMaterialIsStencil=false
     fontMaterial=null
     fontObject=&Roboto
     letterSpacing=0
     monospacedWidth=1
     size=48
     skewAmount=0
     typefaceFontName=Bold
     outlineSettings {
      bApplyOutlineToDropShadows=false
      bMiteredCorners=false
      bSeparateFillAlpha=false
      outlineMaterial=null
      outlineSize=0
      outlineColor=<rgba 0 0 0 1>
     }
    }
    shadowColorAndOpacity=<rgba 0 0 0 1>
   }
   slot {class=VerticalBoxSlot, properties {horizontalAlignment=HAlign_Center, verticalAlignment=VAlign_Center}}
  }
  TextBlock_107 {
   widgetClassPath=&TextBlock
   parent=VerticalBox_121
   widget=TextBlock_107
   properties {
    text=(Low Quality)
    colorAndOpacity {colorUseRule=UseColor_Specified, specifiedColor=<rgba 1 1 1 0.75>}
    shadowColorAndOpacity=<rgba 0 0 0 1>
   }
   slot {class=VerticalBoxSlot, properties {horizontalAlignment=HAlign_Center, verticalAlignment=VAlign_Center}}
  }
  VerticalBox_160 {
   widgetClassPath=&VerticalBox
   parent=VerticalBox_14
   widget=VerticalBox_160
   slot {class=VerticalBoxSlot, properties {verticalAlignment=VAlign_Bottom}}
  }
  Overlay_69 {
   widgetClassPath=&Overlay
   parent=VerticalBox_160
   widget=Overlay_69
   slot {class=VerticalBoxSlot, properties {size {sizeRule=Fill, value=1}}}
  }
  Image_85 {
   widgetClassPath=&Image
   parent=Overlay_69
   widget=Image_85
   properties {colorAndOpacity=<rgba 0 0 0 0.25>}
   slot {class=OverlaySlot, properties {horizontalAlignment=HAlign_Fill, verticalAlignment=VAlign_Fill}}
  }
  Header {
   widgetClassPath=&HorizontalBox
   parent=Overlay_69
   widget=Header
   slot {class=OverlaySlot, properties {horizontalAlignment=HAlign_Fill, verticalAlignment=VAlign_Fill, padding=<ltrb 12 0 12 12>}}
  }
  TextBlock_80 {
   widgetClassPath=&TextBlock
   parent=Header
   widget=TextBlock_80
   properties {
    text=Overall
    font {
     bForceMonospaced=false
     bMaterialIsStencil=false
     fontMaterial=null
     fontObject=&Roboto
     letterSpacing=0
     monospacedWidth=1
     size=42
     skewAmount=0
     typefaceFontName=Bold
     outlineSettings {
      bApplyOutlineToDropShadows=false
      bMiteredCorners=false
      bSeparateFillAlpha=false
      outlineMaterial=null
      outlineSize=0
      outlineColor=<rgba 0 0 0 1>
     }
    }
   }
   slot {class=HorizontalBoxSlot, properties {horizontalAlignment=HAlign_Left}}
  }
  Spacer_78 {
   widgetClassPath=&Spacer
   parent=Header
   widget=Spacer_78
   slot {class=HorizontalBoxSlot, properties {size {sizeRule=Fill, value=1}}}
  }
  TextBlock_2 {
   widgetClassPath=&TextBlock
   parent=Header
   widget=TextBlock_2
   properties {
    text=Current Camera Cut
    font {
     bForceMonospaced=false
     bMaterialIsStencil=false
     fontMaterial=null
     fontObject=&Roboto
     letterSpacing=0
     monospacedWidth=1
     size=42
     skewAmount=0
     typefaceFontName=Bold
     outlineSettings {
      bApplyOutlineToDropShadows=false
      bMiteredCorners=false
      bSeparateFillAlpha=false
      outlineMaterial=null
      outlineSize=0
      outlineColor=<rgba 0 0 0 1>
     }
    }
   }
   slot {class=HorizontalBoxSlot, properties {horizontalAlignment=HAlign_Right}}
  }
  Overlay_70 {
   widgetClassPath=&Overlay
   parent=VerticalBox_160
   widget=Overlay_70
   slot {class=VerticalBoxSlot, properties {verticalAlignment=VAlign_Top}}
  }
  Image_124 {
   widgetClassPath=&Image
   parent=Overlay_70
   widget=Image_124
   properties {colorAndOpacity=<rgba 0 0 0 0.25>}
   slot {class=OverlaySlot, properties {horizontalAlignment=HAlign_Fill, verticalAlignment=VAlign_Fill}}
  }
  HorizontalBox {
   widgetClassPath=&HorizontalBox
   parent=Overlay_70
   widget=HorizontalBox
   slot {class=OverlaySlot, properties {horizontalAlignment=HAlign_Fill, verticalAlignment=VAlign_Fill, padding=<ltrb 12 12 12 12>}}
  }
  VerticalBox_158 {
   widgetClassPath=&VerticalBox
   parent=HorizontalBox
   widget=VerticalBox_158
   slot {class=HorizontalBoxSlot, properties {horizontalAlignment=HAlign_Left, size {sizeRule=Fill, value=1}}}
  }
  CurrentSequenceRow {
   widgetClassPath=&/UI_MovieRenderPipelineInfoTableRow
   bIsVariable=true
   parent=VerticalBox_158
   widget=CurrentSequenceRow
   properties {leftColumnText=Current Sequence:, rightColumnText=Pipeline Test Sequence}
   slot {class=VerticalBoxSlot}
  }
  TotalFrameRow {
   widgetClassPath=&UI_MovieRenderPipelineInfoTableRow
   bIsVariable=true
   parent=VerticalBox_158
   widget=TotalFrameRow
   properties {leftColumnText=Total Frames:, rightColumnText="1245 of 6000 (20%)"}
   slot {class=VerticalBoxSlot}
  }
  ElapsedTimeRow {
   widgetClassPath=&UI_MovieRenderPipelineInfoTableRow
   bIsVariable=true
   parent=VerticalBox_158
   widget=ElapsedTimeRow
   properties {leftColumnText=Elapsed Time:, rightColumnText="[00:06:34]"}
   slot {class=VerticalBoxSlot}
  }
  EstimatedTimeRemainingRow {
   widgetClassPath=&UI_MovieRenderPipelineInfoTableRow
   bIsVariable=true
   parent=VerticalBox_158
   widget=EstimatedTimeRemainingRow
   properties {leftColumnText=Est. Time Remaining:, rightColumnText="[00:12:34]"}
   slot {class=VerticalBoxSlot}
  }
  PipelineStateRow {
   widgetClassPath=&UI_MovieRenderPipelineInfoTableRow
   bIsVariable=true
   parent=VerticalBox_158
   widget=PipelineStateRow
   properties {leftColumnText=Pipeline State:, rightColumnText=Producing Frames}
   slot {class=VerticalBoxSlot}
  }
  VerticalBox_44 {
   widgetClassPath=&VerticalBox
   parent=HorizontalBox
   widget=VerticalBox_44
   slot {class=HorizontalBoxSlot, properties {horizontalAlignment=HAlign_Right, size {sizeRule=Fill, value=1}}}
  }
  ShotNameRow {
   widgetClassPath=&UI_MovieRenderPipelineInfoTableRow
   bIsVariable=true
   parent=VerticalBox_44
   widget=ShotNameRow
   properties {leftColumnText=Cut Name:, rightColumnText="02_0400 (1 of 5)"}
   slot {class=VerticalBoxSlot}
  }
  ShotFrameCountRow {
   widgetClassPath=&UI_MovieRenderPipelineInfoTableRow
   bIsVariable=true
   parent=VerticalBox_44
   widget=ShotFrameCountRow
   properties {leftColumnText=Frame:, rightColumnText="1230 of 3000 (20%)"}
   slot {class=VerticalBoxSlot}
  }
  ShotSubFrameRow {
   widgetClassPath=&UI_MovieRenderPipelineInfoTableRow
   bIsVariable=true
   parent=VerticalBox_44
   widget=ShotSubFrameRow
   properties {leftColumnText=Sub Sample:, rightColumnText=32 of 84}
   slot {class=VerticalBoxSlot}
  }
  ShotEngineWarmUpFramesRow {
   widgetClassPath=&UI_MovieRenderPipelineInfoTableRow
   bIsVariable=true
   parent=VerticalBox_44
   widget=ShotEngineWarmUpFramesRow
   properties {leftColumnText=Engine Warm Up Frames:, rightColumnText=1/4}
   slot {class=VerticalBoxSlot}
  }
  ShotHandleFramesRow {
   widgetClassPath=&UI_MovieRenderPipelineInfoTableRow
   bIsVariable=true
   parent=VerticalBox_44
   widget=ShotHandleFramesRow
   properties {leftColumnText=Handle Frames:, rightColumnText=4 of 12, visibility=Collapsed}
   slot {class=VerticalBoxSlot}
  }
  ShotStateRow {
   widgetClassPath=&UI_MovieRenderPipelineInfoTableRow
   bIsVariable=true
   parent=VerticalBox_44
   widget=ShotStateRow
   properties {leftColumnText=Cut State:, rightColumnText=Uninitialized}
   slot {class=VerticalBoxSlot}
  }
  VerticalBox_62 {widgetClassPath=&VerticalBox, parent=WidgetSwitcher_0, widget=VerticalBox_62, slot {class=WidgetSwitcherSlot}}
  TextBlock_1 {
   widgetClassPath=&TextBlock
   parent=VerticalBox_62
   widget=TextBlock_1
   properties {
    text=Initializing
    font {
     bForceMonospaced=false
     bMaterialIsStencil=false
     fontMaterial=null
     fontObject=&Roboto
     letterSpacing=0
     monospacedWidth=1
     size=48
     skewAmount=0
     typefaceFontName=Bold
     outlineSettings {
      bApplyOutlineToDropShadows=false
      bMiteredCorners=false
      bSeparateFillAlpha=false
      outlineMaterial=null
      outlineSize=0
      outlineColor=<rgba 0 0 0 1>
     }
    }
    shadowColorAndOpacity=<rgba 0 0 0 1>
   }
   slot {
    class=VerticalBoxSlot
    properties {horizontalAlignment=HAlign_Center, verticalAlignment=VAlign_Bottom, padding=<ltrb 0 0 0 64>, size {sizeRule=Fill, value=1}}
   }
  }
 ]
}
