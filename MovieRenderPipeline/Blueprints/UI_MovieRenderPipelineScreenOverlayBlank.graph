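# Blank screen overlay widget for Movie Render Pipeline, derived from MovieRenderDebugWidget, with no widgets of its own, so it adds no visible content.
# Its only logic handles OnInitializedForPipeline by storing the supplied Movie Pipeline in the Pipeline variable, which nothing else in the graph reads.
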WidgetBlueprint /MovieRenderPipeline/Blueprints/UI_MovieRenderPipelineScreenOverlayBlank
graphs [EventGraph {nodes [
   n1 = Event OnInitializedForPipeline!() -> (OutputDelegate, then, ForPipeline:Movie Pipeline Object Reference) {is=title K2Node_Event}
   > Set Pipeline(Pipeline:Movie Pipeline Object Reference=@n1.ForPipeline) -> (Output_Get:Movie Pipeline Object Reference)
  ]}]
variables [Pipeline {type=MoviePipeline, default=null}]
widgets {
 info {inheritedWidgetCount=0, parentClass=/Script/MovieRenderPipelineCore.MovieRenderDebugWidget, rootWidgetClass=null, widgetCount=0}
 widgets []
}
