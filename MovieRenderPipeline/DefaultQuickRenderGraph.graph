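# Default Movie Render Graph config for Quick Render, producing a JPG image sequence under the project's Saved/MovieRenders/QuickRender folder.
# The Globals branch chains warm-up frames with emulated motion blur, game overrides that disable LODs and HLODs and flush streaming and compilers, then global output settings.
# Graph variables feed the output directory and resolution into those settings.
# The DefaultLayer branch runs a deferred renderer into the JPG output and a single render layer.
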MovieGraphConfig /MovieRenderPipeline/DefaultQuickRenderGraph
members {
 inputs [
  MovieGraphInput_0 {class=MovieGraphInput}
  MovieGraphInput_1 {class=MovieGraphInput}
 ]
 outputs [
  MovieGraphOutput_0 {class=MovieGraphOutput}
  MovieGraphOutput_1 {class=MovieGraphOutput}
 ]
 variables [
  MovieGraphVariable_1 {class=MovieGraphVariable}
  MovieGraphVariable_2 {class=MovieGraphVariable}
 ]
}
ids=id
nodes [
 DefaultInputNode = Input() {
  is=title MovieGraphInputNode
  pins are=other
  also {
   outputs [
    MovieGraphPin_1 {properties {type=None, allow_multiple_connections=false, is_branch=true, label=Globals}}
    MovieGraphPin_0 {properties {type=DOUBLE, allow_multiple_connections=false, is_branch=true, label=DefaultLayer}}
   ]
   position [-336,0]
  }
 }
 n1 = Global Game Overrides(MovieGraphPin_0=Warm Up Settings(MovieGraphPin_0=@DefaultInputNode.Globals) {
   is=title MovieGraphWarmUpSettingNode
   pins are=other
   bEmulateMotionBlur=true
   bOverride_NumWarmUpFrames=true
   bOverride_bEmulateMotionBlur=true
   numWarmUpFrames=64
   also {
    inputs [MovieGraphPin_0 {properties {type=None, allow_multiple_connections=false, is_branch=true}}]
    outputs [MovieGraphPin_1 {properties {type=None, allow_multiple_connections=false, is_branch=true}}]
    position [240,0]
   }
  }) {
  is=title MovieGraphGlobalGameOverridesNode
  pins are=other
  bDisableHLODs=true
  bDisableLODs=true
  bFlushAssetCompiler=true
  bFlushGrassStreaming=true
  bFlushLevelStreaming=true
  bFlushShaderCompiler=true
  bFlushStreamingManagers=true
  bOverride_GameModeOverride=true
  bOverride_VirtualTextureFeedbackFactor=true
  bOverride_bDisableHLODs=true
  bOverride_bDisableLODs=true
  bOverride_bDisableTextureStreaming=true
  bOverride_bFlushAssetCompiler=true
  bOverride_bFlushGrassStreaming=true
  bOverride_bFlushLevelStreaming=true
  bOverride_bFlushShaderCompiler=true
  bOverride_bFlushStreamingManagers=true
  also {
   inputs [MovieGraphPin_0 {properties {type=None, allow_multiple_connections=false, is_branch=true}}]
   outputs [MovieGraphPin_1 {properties {type=None, allow_multiple_connections=false, is_branch=true}}]
   position [480,0]
  }
 }
 MovieGraphDeferredRenderPassNode_2 = Deferred Renderer(MovieGraphPin_0=Reroute(MovieGraphPin_0=@DefaultInputNode.DefaultLayer) {
   is=title MovieGraphRerouteNode
   pins are=other
   also {
    inputs [MovieGraphPin_0 {properties {type=None, allow_multiple_connections=false, is_branch=true}}]
    outputs [MovieGraphPin_1 {properties {type=None, allow_multiple_connections=false, is_branch=true}}]
    position [-64,224]
   }
  }) {
  is=title MovieGraphDeferredRenderPassNode
  pins are=other
  showFlags=&self:MovieGraphDeferredRenderPassNode_2.ShowFlags
  additionalPostProcessMaterials []
  also {
   inputs [MovieGraphPin_0 {properties {type=None, allow_multiple_connections=false, is_branch=true}}]
   outputs [MovieGraphPin_1 {properties {type=None, allow_multiple_connections=false, is_branch=true}}]
   position [368,192]
  }
 }
 DefaultOutputNode = Output(
  Globals=Global Output Settings(
   MovieGraphPin_0=@n1
   OutputDirectory=OutputDirectory() {
    is=title MovieGraphVariableNode
    pins are=other
    also {
     outputs [MovieGraphPin_0 {properties {type=STRUCT, allow_multiple_connections=false, label=OutputDirectory, type_object=/Script/CoreUObject.DirectoryPath}}]
     position [528,-176]
    }
   }.OutputDirectory
   OutputResolution=OutputResolution() {
    is=title MovieGraphVariableNode
    pins are=other
    also {
     outputs [MovieGraphPin_0 {properties {type=STRUCT, allow_multiple_connections=false, label=OutputResolution, type_object=/Script/MovieRenderPipelineCore.MovieGraphNamedResolution}}]
     position [528,-96]
    }
   }.OutputResolution
  ) {
   is=title MovieGraphGlobalOutputSettingNode
   pins are=other
   bOverride_OutputDirectory=true
   outputDirectory {path="{project_dir}/Saved/MovieRenders/QuickRender"}
   also {
    inputs [
     MovieGraphPin_0 {properties {type=None, allow_multiple_connections=false, is_branch=true}}
     MovieGraphPin_2 {properties {type=STRUCT, allow_multiple_connections=false, label=OutputDirectory, type_object=&DirectoryPath}}
     MovieGraphPin_4 {properties {type=STRUCT, allow_multiple_connections=false, label=OutputResolution, type_object=&MovieGraphNamedResolution}}
    ]
    outputs [MovieGraphPin_1 {properties {type=None, allow_multiple_connections=false, is_branch=true}}]
    position [736,0]
   }
  }
  DefaultLayer=Render Layer(MovieGraphPin_0=.jpg Sequence(MovieGraphPin_0=@MovieGraphDeferredRenderPassNode_2) {
    is=title MovieGraphImageSequenceOutputNode_JPG
    pins are=other
    also {
     inputs [MovieGraphPin_0 {properties {type=None, allow_multiple_connections=false, is_branch=true}}]
     outputs [MovieGraphPin_1 {properties {type=None, allow_multiple_connections=false, is_branch=true}}]
     position [592,176]
    }
   }) {
   is=title MovieGraphRenderLayerNode
   pins are=other
   bOverride_LayerName=true
   layerName=Layer1
   also {
    inputs [MovieGraphPin_0 {properties {type=None, allow_multiple_connections=false, is_branch=true}}]
    outputs [MovieGraphPin_1 {properties {type=None, allow_multiple_connections=false, is_branch=true}}]
    position [800,176]
   }
  }
 ) {
  is=title MovieGraphOutputNode
  pins are=other
  also {
   inputs [
    MovieGraphPin_1 {properties {type=None, allow_multiple_connections=false, is_branch=true, label=Globals}}
    MovieGraphPin_0 {properties {type=DOUBLE, allow_multiple_connections=false, is_branch=true, label=DefaultLayer}}
   ]
   position [1040,0]
  }
 }
]
