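# Post-process material in Movie Render Pipeline that, going by its name and settings, serves as a stencil-based cutout.
# It runs before depth of field with alpha output enabled, and a greater-or-equal stencil test limits which pixels it affects.
# The graph holds only two constants, zero and one, with no texture or scene inputs, so the stencil test does all the per-pixel selection.
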Material /MovieRenderPipeline/Materials/MoviePipeline_StencilCutout
graph {
 properties {
  bEnableStencilTest=true
  blendableLocation=BL_SceneColorBeforeDOF
  blendableOutputAlpha=true
  materialDomain=MD_PostProcess
  pixelDepthOffsetMode=PDOM_Legacy
  stencilCompare=MSC_GreaterEqual
  usedShadingModels=Default Lit
 }
 nodes [
  Constant()
  Constant() {r=1}
 ]
}
