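# Post-process material for Movie Render Queue that outputs the scene's ambient occlusion, apparently as a render pass.
# It samples the AmbientOcclusion scene texture through a feature level switch.
# The ES2 and ES3_1 feature levels receive a constant white value instead of the texture.
# Negative emissive color is allowed.
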Material /MovieRenderPipeline/Materials/MovieRenderQueue_AmbientOcclusion
graph {
 properties {bAllowNegativeEmissiveColor=true, materialDomain=MD_PostProcess, pixelDepthOffsetMode=PDOM_Legacy, usedShadingModels=Default Lit}
 nodes [
  n1 = SceneTexture() {sceneTextureId=PPI_AmbientOcclusion}
  n2 = Constant() {r=1}
  FeatureLevelSwitch(Default=@n1.Color, ES2=@n2, ES3_1=@n2, SM4_REMOVED=@n1.Color)
 ]
}
