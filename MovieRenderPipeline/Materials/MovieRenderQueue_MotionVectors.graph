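# Post-process material for Movie Render Queue that encodes scene motion vectors as a color.
# Scene velocity is divided by 4 and offset by 0.5 to center it on mid-gray, and its R component feeds both X and Y of the result.
# Z carries the R component of the Custom Stencil scene texture.
# Feature-level and shading-path switches substitute constants on ES2, ES3.1 and forward shading.
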Material /MovieRenderPipeline/Materials/MovieRenderQueue_MotionVectors
graph {
 properties {materialDomain=MD_PostProcess, pixelDepthOffsetMode=PDOM_Legacy, usedShadingModels=Default Lit}
 nodes [
  Constant()
  n1 = SceneTexture() {sceneTextureId=PPI_Velocity}
  n2 = MaterialFunctionCall(Float2=Add(A=Divide(A=@n1.Color) {constB=4}) {constB=0.5}) {materialFunction=&BreakOutFloat2Components}
  n3 = MaterialFunctionCall(
   X=@n2.R
   Y=@n2.R
   Z=MaterialFunctionCall(Float3=SceneTexture() {sceneTextureId=PPI_CustomStencil}.Color) {materialFunction=&BreakOutFloat3Components}.R
  ) {materialFunction=&MakeFloat3}
  n4 = Constant()
  ShadingPathSwitch(Default=FeatureLevelSwitch(Default=@n3.Result, ES2=@n4, ES3_1=@n4, SM4_REMOVED=@n3.Result), Forward=Constant())
 ]
}
