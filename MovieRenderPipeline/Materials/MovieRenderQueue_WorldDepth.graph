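# Post-process material for Movie Render Queue that outputs scene depth and world height as color data.
# Red holds raw SceneDepth, green converts it to radial camera distance by dividing by the negated dot of camera direction and camera vector, and blue holds world position Z.
# Negative emissive is allowed, so values below zero can pass through.
# A feature level switch substitutes a default constant on ES2, ES3_1 and SM4.
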Material /MovieRenderPipeline/Materials/MovieRenderQueue_WorldDepth
graph {
 properties {bAllowNegativeEmissiveColor=true, materialDomain=MD_PostProcess, pixelDepthOffsetMode=PDOM_Legacy, usedShadingModels=Default Lit}
 nodes [
  n1 = MaterialFunctionCall(Float4=SceneTexture() {sceneTextureId=PPI_SceneDepth}.Color) {materialFunction=&BreakOutFloat4Components}
  n2 = Multiply(A=DotProduct(
    A=MaterialFunctionCall() {materialFunction=&CameraDirectionVector}.Result
    B=CameraVectorWS()
   )) {constB=-1}
  n3 = WorldPosition()
  n4 = Constant()
  FeatureLevelSwitch(
   Default=MaterialFunctionCall(
    X=@n1.R
    Y=Divide(A=@n1.R, B=@n2)
    Z=MaterialFunctionCall(Float3=@n3.XYZ) {materialFunction=&BreakOutFloat3Components}.B
   ) {materialFunction=&MakeFloat4}.Result
   ES2=@n4
   ES3_1=@n4
   SM4_REMOVED=@n4
  )
 ]
}
