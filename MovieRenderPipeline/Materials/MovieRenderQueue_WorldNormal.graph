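# Post-process material for Movie Render Queue that outputs the scene's world normals as a color image.
# It reads the WorldNormal scene texture and scales and offsets it by one half, remapping signed normals into the 0-to-1 range.
# A shading path switch outputs constant black under forward shading, and a feature level switch substitutes constant white on ES2, ES3.1 and SM4.
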Material /MovieRenderPipeline/Materials/MovieRenderQueue_WorldNormal
graph {
 properties {bAllowNegativeEmissiveColor=true, materialDomain=MD_PostProcess, pixelDepthOffsetMode=PDOM_Legacy, usedShadingModels=Default Lit}
 nodes [
  n1 = Multiply(A=SceneTexture() {sceneTextureId=PPI_WorldNormal}.Color) {constB=0.5}
  n2 = Constant() {r=1}
  ShadingPathSwitch(Default=FeatureLevelSwitch(Default=Add(A=@n1) {constB=0.5}, ES2=@n2, ES3_1=@n2, SM4_REMOVED=@n2), Forward=Constant())
 ]
}
