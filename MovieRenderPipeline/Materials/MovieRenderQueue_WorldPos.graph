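# Post-process material for Movie Render Queue that outputs world position, by its name as a render pass.
# A feature level switch passes world position XYZ by default and substitutes a constant on the ES2, ES3.1 and removed SM4 inputs.
# Negative emissive color is allowed, so negative coordinates can be written without being clamped to zero.
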Material /MovieRenderPipeline/Materials/MovieRenderQueue_WorldPos
graph {
 properties {bAllowNegativeEmissiveColor=true, materialDomain=MD_PostProcess, pixelDepthOffsetMode=PDOM_Legacy, usedShadingModels=Default Lit}
 nodes [
  n1 = Constant() {r=1}
  FeatureLevelSwitch(Default=WorldPosition().XYZ, ES2=@n1, ES3_1=@n1, SM4_REMOVED=@n1)
 ]
}
